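# Additive, unlit, two-sided material for Niagara mesh particles that draws the Thermite area-of-effect dome as a flame-textured shell.
# Custom atan2 on vertex normals gives U and time-panned world Z gives V, wrapping a scrolling flame texture tinted by particle color.
# Two Dynamic Parameter-driven Fresnels, sharpened by repeated squaring, shape a rim band scaled by particle alpha.
# A normal-based mask hides the poles, while depth and near-camera fades soften intersections.
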Material /ThermiteGameplay/Assets/FX/Materials/M_Thermite_AoE_Dome
graph {
 properties {
  bUseTranslucencyVertexFog=false
  bUsedWithNiagaraMeshParticles=true
  blendMode=BLEND_Additive
  pixelDepthOffsetMode=PDOM_Legacy
  shadingModel=MSM_Unlit
  twoSided=true
  usedShadingModels=Unlit
 }
 nodes [
  MaterialExpressionComponentMask_7 = ComponentMask(None=VertexNormalWS()) {g=true, r=true}
  n1 = Custom!(InputUV=@MaterialExpressionComponentMask_7) {
   code="float2 UV =InputUV;\r\nreturn atan2(UV.y, UV.x);"
   showCode=true
   additionalOutputs [{outputName=None, outputType=CMOT_Float2}]
   inputs [{inputName=InputUV, input {expression=&self:MaterialExpressionComponentMask_7}}]
  }
  n2 = Constant() {r=3.1416}
  n3 = Divide(A=Add(A=@n1.return, B=@n2) {constB=3.14}, B=@n2) {constB=3.14}
  n4 = Add(A=WorldPosition().Z, B=Multiply(A=Time(), B=ScalarParameter() {defaultValue=155, parameterName=Texture Pan Speed}))
  n5 = AppendVector(
   A=Multiply(A=ComponentMask(None=@n3) {r=true}, B=ScalarParameter() {defaultValue=2, parameterName=U_Scale}) {constB=0.314}
   B=Multiply(A=Multiply(A=@n4) {constB=0.001}, B=ScalarParameter() {defaultValue=2, parameterName=V_Scale})
  )
  n6 = ParticleColor()
  MaterialFunctionCall(Emissive=Multiply(
    A=TextureSample(UVs=@n5) {samplerType=SAMPLERTYPE_LinearGrayscale, texture=/Game/Effects/Fort_Effects/Emotes/UltraLight/Texture/T_FluidFlameTile_Greyscale}.RGB
    B=@n6.RGB
   )) {materialFunction=&MF_FX_EmissiveAdjustments}
  n7 = NamedRerouteUsage()
  n8 = DynamicParameter() {defaultValue=<rgba 1 0.6 1 1>, paramNames [FresnelExp, FresnelOuterExp, Param3, Param4]}
  n9 = Multiply(A=@n6.A, B=Fresnel(ExponentIn=@n8.FresnelExp))
  n10 = Multiply(A=@n9, B=@n9)
  n11 = Multiply(A=@n10, B=@n10)
  n12 = Fresnel(ExponentIn=@n8.FresnelOuterExp) {baseReflectFraction=-1}
  n13 = Subtract(A=OneMinus(None=Multiply(A=@n12, B=@n6.A))) {constB=0.1}
  n14 = Multiply(A=@n13, B=@n13)
  n15 = Multiply(A=@n14, B=@n14)
  n16 = Multiply(A=Multiply(A=Saturate(None=@n11), B=Saturate(None=@n15))) {constB=10}
  Multiply(
   A=MaterialFunctionCall(Distance=ScalarParameter() {defaultValue=250, parameterName=Near Camera Fade}) {materialFunction=&Near_Camera_Fade}.Result
   B=DepthFade(Opacity=Multiply(A=@n7, B=@n16), FadeDistance=ScalarParameter() {defaultValue=100, parameterName=Depth Fade})
  )
  n17 = DotProduct(A=Constant3Vector() {constant=<rgba 0 0 1 0>}, B=VertexNormalWS())
  n18 = Multiply(A=OneMinus(None=Abs(None=@n17))) {constB=5}
  NamedRerouteDeclaration(None=Saturate(None=Multiply(A=@n18, B=@n18))) {name=UPDOWN Mask, nodeColor=<rgba 0 0.4609 1 1>}
 ]
}
