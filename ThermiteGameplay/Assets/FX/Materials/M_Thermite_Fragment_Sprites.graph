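# Translucent unlit material for Niagara mesh particles that renders thermite fragments as soft, round, camera-facing sprites.
# The vertex-position math rebuilds mesh quads along camera axes, with size varied by noise, vertex color and dynamic parameters.
# A turbulent motion term combines animated vertex-stage noise, a dynamic XYZ vector and vertex-color-weighted fake gravity.
# Particle color drives emissive, while opacity uses a sphere mask and depth fade, with a near-camera fade also defined.
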Material /ThermiteGameplay/Assets/FX/Materials/M_Thermite_Fragment_Sprites
graph {
 properties {
  bUsedWithNiagaraMeshParticles=true
  blendMode=BLEND_Translucent
  opacityMaskClipValue=0.05
  pixelDepthOffsetMode=PDOM_Legacy
  shadingModel=MSM_Unlit
  twoSided=true
  usedShadingModels=Unlit
 }
 nodes [
  n1 = WorldPosition()
  n2 = ObjectPositionWS()
  n3 = VertexColor()
  n4 = TextureSample(UVs=Add(A=TextureCoordinate(), B=ComponentMask(None=@n3) {b=true, g=true})) {samplerType=SAMPLERTYPE_LinearColor, texture=/Game/Effects/Fort_Effects/Gameplay/SmokeDustExplosions/Textures/T_NoisePack_LinearColor}
  n5 = MaterialFunctionCall(Float2=Add(A=ComponentMask(None=@n4.RGB) {b=true, g=true}, B=TextureCoordinate())) {materialFunction=&BreakOutFloat2Components}
  n6 = Constant3Vector() {constant=<rgba 1 0 0 0>}
  n7 = Add(A=Multiply(A=@n3.G) {constB=0.5}) {constB=0.5}
  n8 = DynamicParameter() {defaultValue=<rgba 1 1500 0 1>, paramNames [Quad Scale, Anim Scale, Anim Speed, Fake Gravity]}
  n9 = Multiply(A=ScalarParameter() {defaultValue=5, parameterName=ScaleX}, B=@n8.Quad Scale)
  n10 = ScalarParameter() {defaultValue=5, parameterName=ScaleY}
  n11 = MaterialFunctionCall(Float2=AppendVector(A=Multiply(A=@n7, B=@n9), B=Multiply(A=@n10, B=@n8.Quad Scale))) {materialFunction=&BreakOutFloat2Components}
  n12 = Multiply(A=Multiply(A=@n5.R, B=Transform(None=@n6) {transformSourceType=TRANSFORMSOURCE_Camera}), B=@n11.R)
  n13 = Constant3Vector() {constant=<rgba 0 -1 0 0>}
  n14 = Multiply(A=Multiply(A=@n5.G, B=Transform(None=@n13) {transformSourceType=TRANSFORMSOURCE_Camera}), B=@n11.G)
  Add(A=Add(A=Subtract(A=@n1.XYZ, B=@n2), B=Add(A=@n12, B=@n14)), B=NamedRerouteUsage()) {constB=0}
  n15 = ParticleColor()
  MaterialFunctionCall(Emissive=Multiply(A=@n15.RGB, B=NamedRerouteUsage())) {materialFunction=&MF_FX_EmissiveAdjustments}
  n16 = VertexColor()
  n17 = Multiply(
   A=SphereMask(A=TextureCoordinate(), B=Constant() {r=0.5}, Radius=Constant() {r=0.4}) {hardnessPercent=16}
   B=Add(A=Multiply(A=@n16.G) {constB=0.5}) {constB=0.5}
  )
  Multiply(
   A=DepthFade(Opacity=Multiply(A=@n15.A, B=@n17), FadeDistance=ScalarParameter() {defaultValue=15, parameterName=Depth Fade}) {fadeDistanceDefault=15}
   B=NamedRerouteUsage()
  )
  n18 = VertexColor()
  n19 = Multiply(A=Subtract(A=WorldPosition().XYZ, B=ObjectPositionWS()), B=ScalarParameter() {defaultValue=0.001, parameterName=Noise UV Scale}) {constB=0.01}
  n20 = Multiply(
   A=Time()
   B=ComponentMask(None=VectorParameter() {parameterName=Noise UVW Anim Speed, defaultValue=<rgba 0.015 0.015 0.01 0>}.RGB) {b=true, g=true, r=true}
  )
  n21 = TextureSample!(UVs=Add(A=@n18, B=Add(A=@n19, B=@n20)), Tex, Level, Apply View MipBias) {
   automaticViewMipBias=false
   constMipValue=0
   mipValueMode=TMVM_MipLevel
   samplerType=SAMPLERTYPE_LinearColor
   texture=/Game/Gadgets/Assets/VinderTech_GliderChute/Glider_Mitosis_Male/Textures/Noise/VT_GaussianNoise
  }
  n22 = Add(A=Subtract(A=@n21.RGB) {constB=0.5}, B=DynamicParameter() {parameterIndex=1, defaultValue=<rgba 0 0 0 1>, paramNames [X,Y,Z,none]}.RGB)
  n23 = DynamicParameter() {defaultValue=<rgba 1 1500 0 1>, paramNames [Quad Scale, Anim Scale, Anim Speed, Fake Gravity]}
  n24 = Multiply(
   A=Multiply(A=Multiply(A=@n22, B=@n23.Anim Scale), B=VectorParameter() {parameterName=Anim Scale Adjustment, defaultValue=<rgba 1 1 1 0>}.RGB)
   B=@n23.Anim Speed
  )
  n25 = Constant()
  n26 = VertexColor()
  n27 = Multiply(
   A=@n23.Fake Gravity
   B=MaterialFunctionCall(X=@n25, Y=@n25, Z=@n26.B) {materialFunction=&MakeFloat3}.Result
  )
  n28 = MaterialFunctionCall(X=Add(A=Multiply(A=@n26.G) {constB=0.2}) {constB=0.8}, Y=@n25, Z=@n25) {materialFunction=&MakeFloat3}
  NamedRerouteDeclaration(None=Add(A=Add(A=@n24, B=@n27), B=Transform(None=@n28.Result) {transformSourceType=TRANSFORMSOURCE_Local})) {name=Turbulent Anim, nodeColor=<rgba 1 0 0.6797 1>}
  n29 = MaterialFunctionCall(Distance=ScalarParameter() {defaultValue=300, parameterName=Near Camera Fade Distance}) {materialFunction=&Near_Camera_Fade}
  NamedRerouteDeclaration(None=VertexInterpolator(VS=Multiply(A=@n29.Result, B=@n29.Result)).PS) {name=Near Camera Fade, nodeColor=<rgba 0.6875 1 0 1>}
  NamedRerouteDeclaration(None=Add(A=Multiply(A=@n18.R) {constB=0.7}) {constB=0.3}) {name=Noise Channel Red, nodeColor=<rgba 0 1 0.4844 1>}
 ]
}
