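# Niagara system for the Thermite area-of-effect burn: flung molten splatter, fire smoke, embers, lightweight LW_ sparks and flash, plus a flickering light with ground decal.
# Splatter and embers pair GPU high/epic emitters with medium/low variants.
# Splatter sticks on collision through extreme friction, switches effects by speed, and expires by distance from the owner.
# User.IsTrain disables splatter spawning and zeroes decal size.
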NiagaraSystem /ThermiteGameplay/Assets/FX/NS_Thermite_AoE
dependencies {
 usedDataInterfaces [
  &NiagaraDataInterfaceCurve
  &NiagaraDataInterfaceArrayFloat
  &NiagaraDataInterfaceColorCurve
  &NiagaraDataInterfaceSpriteRendererInfo
 ]
 usedDynamicInputs [/Niagara/DynamicInputs/Vectors/RotateVector, &Divide_Float]
 usedModules [
  &self:SystemSpawnScript.SystemScriptSource.SystemScriptGraph.NiagaraNodeAssignment_0.SetVariables_#1
  &SystemState
  &self:SystemSpawnScript.SystemScriptSource.SystemScriptGraph.NiagaraNodeAssignment_1.SetVariables_#2
 ]
}
emitters {
 Epic_High {
  simTarget=GPUComputeSim
  emitterUpdateScript [
   &EmitterState(Inactive Response=::NewEnumerator1, Life Cycle Mode=::NewEnumerator1, Loop Delay=0.14, Loop Duration=1, UseLoopDelay=true)
   &SpawnRate(
    Spawn Probability=&MakeCustomFloatFromBool(Bool=$User.IsTrain, True Float=0, False Float=1)
    SpawnRate=15
    Use Spawn Probability=true
   )
  ]
  particleSpawnScript [
   &InitializeParticle(
    Color=<rgba 2 0.2962 0 1>
    Lifetime Max=1.5
    Lifetime Mode=::NewEnumerator1
    Mesh Scale Mode=::NewEnumerator3
    Mesh Uniform Scale=&Multiply_Float(A=&RandomRangeFloat(Minimum=0.18, Maximum=0.21), B=15)
    Position=$Engine.Owner.Position
    Position Mode=::NewEnumerator1
    Sprite Rotation Angle=&RandomRangeFloat(Minimum=-10, Maximum=10)
    Sprite Rotation Mode=::NewEnumerator0
    Sprite Size Max=<70 50>
    Sprite Size Min=<65 45>
    Sprite Size Mode=::NewEnumerator2
    Sprite UV Mode=::NewEnumerator3
    UsePositionOffset=false
    Write Material Random=true
   ) {module=InitializeParticle001}
   &ShapeLocation(
    Handle Radius=35
    Large Radius=20
    MaskSpawnGroup=false
    Non Uniform Scale=<1 1 0>
    Offset=<0 0 25>
    Offset Coordinate Space=::Local
    Offset Mode=::NewEnumerator0
    Override Seed=false
    Rotation Axis=<0 0 1>
    Rotation Mode=::NewEnumerator1
    Shape Primitive=::NewEnumerator3
    UseOwnerScale=true
   )
   /SetVariables(Particles.SpriteMesh_Quat:Quat4f=/Niagara/DynamicInputs/Rotation/RandomQuaternion(), Particles.Sprite_CamOfs=25) {isSetParameters=true, module=SetVariables_#3}
   &InitialMeshOrientation(
    Mesh Orientation Mode=::NewEnumerator1
    Rotation=&RandomRangeVector(Minimum=<-0.1 -0.1 -0.1>, Maximum=<0.1 0.1 0.1>)
    Rotation Coordinate Space=::NewEnumerator2
    Use Rotation Vector=true
   )
   /SetVariables(
    Particles.InitPos:NiagaraPosition=$Particles.Initial.Position
    Particles.StoredVector:Vector3f=&TransformVector(
     Vector=Linked() {linkedVariable {name=Particles.Initial.Position, type=NiagaraPosition}}
     Source Space=::Simulation
     Destination Space=::World
     Apply Scale=true
     Transform Vector as Position=true
    )
   ) {isSetParameters=true, module=SetVariables_#4}
   &AddVelocity(
    Module.Use Falloff Distance=false
    Origin Offset=&MakeVector(X=0, Y=0, Z=-6)
    Override Seed=false
    Velocity Mode=::NewEnumerator1
    Velocity Origin=$Engine.Owner.Position
    Velocity Speed=&RandomRangeFloat(Minimum=1000, Maximum=2000)
   )
  ]
  particleUpdateScript [
   &ParticleState()
   &UpdateMeshOrientation(Rotation Rate=&RandomRangeFloat(Minimum=-1, Maximum=1), Rotation Vector=<0.15 0.15 0.1>)
   &GravityForce(Gravity=&Multiply_VectorByFloat(
     Vector=<0 0 -15000>
     Float=&MakeCustomFloatFromBool(
      Bool=&SetBoolByFloatComparison(A=15, B=HlslExpression() {hlslExpression="length(Particles.Velocity)"})
      True Float=0
      False Float=1
     )
    ))
   &Collision(
    Age Colliding Particles=false
    Collision Enabled=&SetBoolByFloatComparison(
     A=1
     B=&FloatFromCurve(FloatCurve=DataInterface() {propertyValues {Curve {keys [<at 0.05 0.95>,<at 0.5 2>]}}}, Scale Curve=0.5)
    )
    Enable Rest State=false
    Friction=10000
    Friction During a Bounce=10000
    GPU Collision Type=::NewEnumerator2
    Method for Calculating Particle Radius:ENiagaraCalculateRadiusOptions=::NewEnumerator2
    Module.EnableMaxCollisionCount=false
    Module.Randomize Collision Normal Vector=false
    Particle Radius Scale=0.12
    Randomize Collision Normal=0
    Restitution=0
   )
   &SolveForcesAndVelocity(Clamp Velocity=false, Limit Acceleration=false)
   &ScaleSpriteSize(
    Uniform Curve Scale=3
    Uniform Curve Sprite Scale=DataInterface() {propertyValues {Curve {keys [{tangentMode=RCTM_User, time=0, value=0.2}, <at 0.5 1>, {tangentMode=RCTM_User, time=0.7, value=0}]}}}
   )
   &DynamicMaterialParameters(
    Index 0 Param 1=&MakeFloatFromVector(Vector=$Particles.StoredVector, Channel=::NewEnumerator0)
    Index 0 Param 2=&MakeFloatFromVector(Vector=$Particles.StoredVector, Channel=::NewEnumerator1)
    Index 0 Param 3=&MakeFloatFromVector(Vector=$Particles.StoredVector, Channel=::NewEnumerator2)
    Index 0 Param 4=&FloatFromCurve(
     FloatCurve=DataInterface() {propertyValues {Curve {keys [{tangentMode=RCTM_User, time=0.025, value=1}, {tangentMode=RCTM_User, time=0.6, value=1}]}}}
     Scale Curve=20
    )
    Module.Param0WriteEnabled=true
    Module.Param1WriteEnabled=true
    Module.Param2WriteEnabled=true
    Module.Param3WriteEnabled=true
   ) {module=DynamicMaterialParameters001}
   &ScaleColor(
    Scale Alpha=&MakeCustomFloatFromBool(
     Bool=&SetBoolByFloatComparison(A=10, B=HlslExpression() {hlslExpression="length(Particles.Velocity)"})
     True Float=&FloatFromCurve(
      FloatCurve=DataInterface() {propertyValues {Curve {keys [
          <at 0.1 0.65 RCIM_Cubic>
          {arriveTangent=-0.0405, interpMode=RCIM_Cubic, leaveTangent=-0.0405, tangentMode=RCTM_User, time=0.2, value=1}
          {interpMode=RCIM_Cubic, leaveTangent=1.75, time=0.4, value=0}
         ]}}}
      Scale Curve=&FNTimeOfDayScalar(Night Scale=0.6, Morning Scale=0.7, Evening Scale=0.7)
     )
     False Float=0
    )
    ScaleA=true
    ScaleRGB=false
   )
   /SetVariables(
    Particles.Color2=&Multiply_LinearColorByFloat(
     Linear Color=&ColorFromCurve(ColorCurve=DataInterface() {propertyValues {
        AlphaCurve {keys [{arriveTangent=1.1089, interpMode=RCIM_Cubic, time=0.107, value=15}, <at 0.9427 1>, <at 0.9661 0 RCIM_Cubic>]}
        BlueCurve {keys [<at 0 0 RCIM_Cubic>]}
        GreenCurve {keys [<at 0 1.7498 RCIM_Cubic>]}
        RedCurve {keys [<at 0 10 RCIM_Cubic>]}
       }})
     Float=1
    )
    Particles.Color3:LinearColor=&ColorFromCurve(ColorCurve=DataInterface() {propertyValues {
       AlphaCurve {keys [<at 0.0514 15 RCIM_Cubic>,<at 0.8816 0 RCIM_Cubic>]}
       BlueCurve {keys [<at 0 0 RCIM_Cubic>]}
       GreenCurve {keys [<at 0 52.2235 RCIM_Cubic>]}
       RedCurve {keys [<at 0 200 RCIM_Cubic>]}
      }})
    Particles.DM2001:Vector4f=&MakeFloat4FromVectorAndFloat("Vector (RGB)"=&TransformVector(Vector=<0 0 0>, Destination Space=::World), "Float (Alpha)"=1)
    Particles.DM2_0:Vector4f=&MakeVector4(
     X=&Multiply_Float(
      A=&FloatFromCurve(
       FloatCurve=DataInterface() {propertyValues {Curve {keys [{tangentMode=RCTM_User, time=0, value=1}, {tangentMode=RCTM_User, time=0.8, value=0}]}}}
       Scale Curve=4
      )
      B=&MakeCustomFloatFromBool(
       Bool=&SetBoolByFloatComparison(A=10, B=HlslExpression() {hlslExpression="length(Particles.Velocity)"})
       True Float=0.2
       False Float=0
      )
     )
     Y=&FloatFromCurve(
      FloatCurve=DataInterface() {propertyValues {Curve {keys [
          {arriveTangent=2.2144, interpMode=RCIM_Cubic, leaveTangent=2.2144, tangentMode=RCTM_User, time=0, value=0}
          {tangentMode=RCTM_User, time=0.9, value=1}
         ]}}}
      Scale Curve=8
     )
     Z=&FloatFromCurve(
      FloatCurve=DataInterface() {propertyValues {Curve {keys [
          {arriveTangent=3.2315, interpMode=RCIM_Cubic, leaveTangent=3.2315, tangentMode=RCTM_User, time=0, value=0}
          {arriveTangent=1.1179, interpMode=RCIM_Cubic, leaveTangent=1.1179, tangentMode=RCTM_User, time=0.4, value=1}
         ]}}}
      Scale Curve=45
     )
     W=&FloatFromCurve(
      FloatCurve=DataInterface() {propertyValues {Curve {keys [
          {arriveTangent=0.6176, interpMode=RCIM_Cubic, leaveTangent=0.6176, tangentMode=RCTM_User, time=0.2, value=0}
          {arriveTangent=6.707, interpMode=RCIM_Cubic, leaveTangent=6.707, tangentMode=RCTM_User, time=1, value=2.5}
         ]}}}
      Scale Curve=-250
     )
    )
    Particles.Mesh Scale2:Vector3f=<10 10 10>
    Particles.SplashSize:Vector2f=&Multiply_Vector2D_ByFloat(
     Float=&FloatFromCurve(FloatCurve=DataInterface() {propertyValues {Curve {keys [{tangentMode=RCTM_User, time=0.1, value=0.02}, {tangentMode=RCTM_User, time=0.8, value=2}]}}})
     Vector2D=<5 5>
    )
    Particles.Vector 2D:Vector2f=&Multiply_Vector2D_ByFloat(
     Float=&MakeCustomFloatFromBool(
      Bool=&SetBoolByFloatComparison(A=15, B=HlslExpression() {hlslExpression="length(Particles.Velocity)"})
      True Float=0
      False Float=1
     )
     Vector2D=&RandomRangeVector2D(Minimum=<9 12>, Maximum=<14 35>)
    )
    Particles.Vector 4:Vector4f=&MakeVector4(
     X=&FloatFromCurve(FloatCurve=DataInterface() {propertyValues {Curve {keys [{arriveTangent=-1.7191, interpMode=RCIM_Cubic, leaveTangent=-1.7191, tangentMode=RCTM_User, time=0, value=1}, <at 0.7 0>]}}})
     Y=0
     Z=0
     W=0
    )
   ) {isSetParameters=true}
   /SetVariables(Particles.Lifetime=&MakeCustomFloatFromBool(
     Bool=&SetBoolByFloatComparison(A=450, B=&DistanceBetweenPositions(A=$Engine.Owner.Position, B=$Particles.Position))
     True Float=$Particles.Lifetime
     False Float=0
    )) {isSetParameters=true}
   &SubUVAnimation(
    Evaluation Type=::NewEnumerator0
    "Sprite Renderer "=DataInterface() {propertyValues {SpriteRenderer=&self:NiagaraEmitter_3.NiagaraSpriteRendererProperties_4}}
    SubUV Animation Mode:ENiagara_SubUVLookupModeV2=::NewEnumerator2
    UseEndFrame=false
    UseStartFrame=false
   )
  ]
  rendererClasses [&NiagaraSpriteRendererProperties, &NiagaraMeshRendererProperties]
  renderers [
   {rendererClass=&NiagaraSpriteRendererProperties}
   {rendererClass=&NiagaraMeshRendererProperties, rendererIndex=1}
   {rendererClass=&NiagaraSpriteRendererProperties, rendererIndex=2}
   {rendererClass=&NiagaraSpriteRendererProperties, rendererIndex=3}
  ]
 }
 Fire_Smoke {
  emitterUpdateScript [
   &EmitterState(Delay First Loop Only=true, Life Cycle Mode=::NewEnumerator1, Loop Delay=0.05, Loop Duration=0.22, UseLoopDelay=true)
   &SpawnRate(SpawnRate=11, Use Spawn Probability=false)
  ]
  particleSpawnScript [
   &InitializeParticle(
    Color=<rgba 1 0.9283 0.8553 1>
    Lifetime=&UniformRangedFloat(Minimum=0.9, Maximum=1.7)
    Mass Mode=::NewEnumerator1
    Position=/Niagara/DynamicInputs/Helpers/SimulationPosition()
    Position Mode=::NewEnumerator1
    Position Offset=<0 0 10>
    Sprite Size=&Multiply_Vector2D_ByFloat(
     Float=&FloatFromCurve(
      FloatCurve=DataInterface() {propertyValues {Curve {keys [{arriveTangent=-3.8412, interpMode=RCIM_Cubic, leaveTangent=-3.8412, tangentMode=RCTM_User, time=0, value=2}, <at 0.5 1>]}}}
      CurveIndex=$Emitter.Age
      Scale Curve=&Sine(Normalized Angle=$Emitter.Age, Period=0.9, Scale=0.1, Bias=1)
     )
     Vector2D=&Vector2DFromFloat(Value=&RandomRangeFloat(Minimum=225, Maximum=275))
    )
    Sprite Size Mode=::NewEnumerator4
    Sprite UV Mode=::NewEnumerator1
    UsePositionOffset=true
    Write Material Random=true
   ) {module=InitializeParticle001}
   /SetVariables(Particles.InitSpriteRotRate=&UniformRangedFloat(Minimum=-3, Maximum=3)) {isSetParameters=true}
   &SphereLocation(
    MaskSpawnGroup=false
    Override Seed=false
    Sphere Radius=0
    Surface Only=false
    UseNonUniformScale=false
    UseOffset=true
    UseSphereOrientation=false
   ) {enabled=false}
   &AddVelocity(
    Cone Angle=5
    Cone Axis=<0 0 1>
    Module.UseConeFalloff=false
    Module.UseVelDistribution=false
    Override Seed=false
    Velocity Mode=::NewEnumerator2
    Velocity Speed=&FloatFromCurve(
     FloatCurve=DataInterface() {propertyValues {Curve {keys [{arriveTangent=-2.9253, interpMode=RCIM_Cubic, leaveTangent=-2.9253, tangentMode=RCTM_User, time=0, value=1}, <at 0.3 0.5>]}}}
     CurveIndex=$Emitter.LoopedAge
     Scale Curve=1000
    )
   )
  ]
  particleUpdateScript [
   &ParticleState()
   &SpriteRotationRate(Rotation Rate=&Multiply_Float(
     A=$Particles.InitSpriteRotRate
     B=&FloatFromCurve(
      FloatCurve=DataInterface() {propertyValues {Curve {keys [
          {arriveTangent=-1.9214, interpMode=RCIM_Cubic, leaveTangent=-1.9214, tangentMode=RCTM_User, time=0.2, value=1.0113}
          {arriveTangent=-0.2496, interpMode=RCIM_Cubic, leaveTangent=-0.2496, tangentMode=RCTM_User, time=0.85, value=0.1}
         ]}}}
      Scale Curve=60
     )
    ))
   &ScaleColor(
    Scale Alpha=&FloatFromCurve(FloatCurve=DataInterface() {propertyValues {Curve {keys [
         <at 0 0.45>
         <at 0.35 1.05>
         {arriveTangent=-0.0528, interpMode=RCIM_Cubic, leaveTangent=-0.0528, tangentMode=RCTM_User, time=0.8, value=0.875}
         {interpMode=RCIM_Cubic, tangentMode=RCTM_Break, time=1, value=0}
        ]}}})
    ScaleA=true
    ScaleRGB=false
   )
   &Drag(
    Drag=&FloatFromCurve(FloatCurve=DataInterface() {propertyValues {Curve {keys [<at 0.1 1 RCIM_Cubic>,<at 0.75 0 RCIM_Cubic>]}}}, Scale Curve=3.5)
    Use Linear Drag=true
    Use Rotational Drag=false
   )
   &GravityForce(Gravity=<0 0 50>)
   &CurlNoiseForce(
    Noise Frequency=250
    Noise Quality / Cost=::NewEnumerator0
    Noise Strength=&FloatFromCurve(
     FloatCurve=DataInterface() {propertyValues {Curve {keys [{tangentMode=RCTM_User, time=0.2, value=0}, {tangentMode=RCTM_User, time=0.6, value=1}]}}}
     Scale Curve=&RandomRangeFloat(Minimum=1500, Maximum=2500)
    )
    Pan Noise=true
    Pan Noise Field=<0.6 0.6 1>
    Randomization Vector:Vector3f=<0.65 0.4 0.05>
   )
   &SolveForcesAndVelocity(Clamp Velocity=false, Limit Acceleration=false)
   &ScaleSpriteSize(Scale Factor=&Vector2DFromFloat(Value=&FloatFromCurve(FloatCurve=DataInterface() {propertyValues {Curve {keys [
          {arriveTangent=0.8986, interpMode=RCIM_Cubic, leaveTangent=0.8986, tangentMode=RCTM_User, time=0.02, value=0.1}
          {arriveTangent=2.5631, interpMode=RCIM_Cubic, leaveTangent=2.5631, tangentMode=RCTM_User, time=0.2, value=0.5}
          <at 0.7 1 RCIM_Cubic>
         ]}}})))
   &DynamicMaterialParameters(
    Index 0 Param 1=&Add_Float(
     A=&RandomRangeFloat(Minimum=0, Maximum=5)
     B=&FloatFromCurve(
      FloatCurve=DataInterface() {propertyValues {Curve {keys [
          {arriveTangent=-2.2561, interpMode=RCIM_Cubic, leaveTangent=-2.2561, tangentMode=RCTM_User, time=0, value=1}
          {arriveTangent=-1.0036, interpMode=RCIM_Cubic, leaveTangent=-1.0036, tangentMode=RCTM_User, time=0.2682, value=0.3}
          <at 0.8 0>
         ]}}}
      Scale Curve=0.5
     )
    )
    Index 0 Param 2=&FloatFromCurve(
     FloatCurve=DataInterface() {propertyValues {Curve {keys [
         {arriveTangent=1.5102, interpMode=RCIM_Cubic, leaveTangent=1.5102, tangentMode=RCTM_User, time=0.15, value=0.5}
         <at 0.8 0.75 RCIM_Cubic>
         {arriveTangent=-1.4584, interpMode=RCIM_Cubic, leaveTangent=-1.4584, tangentMode=RCTM_User, time=1, value=0.5}
        ]}}}
     Scale Curve=0.4
    )
    Index 0 Param 3=1
    Index 0 Param 4=1
    Index 1 Param 1=&FloatFromCurve(
     FloatCurve=DataInterface() {propertyValues {Curve {keys [
         {arriveTangent=-2.9951, interpMode=RCIM_Cubic, leaveTangent=-2.9951, tangentMode=RCTM_User, time=0.1, value=0.5}
         {arriveTangent=-0.5467, interpMode=RCIM_Cubic, leaveTangent=-0.5467, tangentMode=RCTM_User, time=0.5, value=0}
        ]}}}
     Scale Curve=&FloatFromCurve(FloatCurve=DataInterface() {propertyValues {Curve {keys [<at 0.1 15>,<at 0.3 1>,<at 1 1>]}}})
    )
    Index 1 Param 2=&FloatFromCurve(
     FloatCurve=DataInterface() {propertyValues {Curve {keys [{arriveTangent=-1.2376, interpMode=RCIM_Cubic, leaveTangent=-1.2376, tangentMode=RCTM_User, time=0, value=1}, <at 1 0>]}}}
     Scale Curve=10
    )
    Index 1 Param 3=&FloatFromCurve(FloatCurve=DataInterface() {propertyValues {Curve {keys [
         {interpMode=RCIM_Cubic, tangentMode=RCTM_User, time=0.333, value=1.5}
         {arriveTangent=0.7571, interpMode=RCIM_Cubic, leaveTangent=0.7571, tangentMode=RCTM_User, time=1, value=1.7}
        ]}}})
    Index 1 Param 4=&FloatFromCurve(
     FloatCurve=DataInterface() {propertyValues {Curve {keys [<at 0 0.5>,<at 1 0.3>]}}}
     Scale Curve=&Add_Float(A=1, B=&RandomRangeFloat(Minimum=0, Maximum=5))
    )
    Index 2 Param 1=&FloatFromCurve(FloatCurve=DataInterface() {propertyValues {Curve {keys [{arriveTangent=0.0048, interpMode=RCIM_Cubic, leaveTangent=0.0048, tangentMode=RCTM_User, time=0, value=0.8}, <at 0.9 1.2 RCIM_Cubic>]}}})
    Index 2 Param 2=&Add_Float(
     A=&RandomRangeFloat(Minimum=0, Maximum=100)
     B=&FloatFromCurve(
      FloatCurve=DataInterface() {propertyValues {Curve {keys [
          {arriveTangent=0.5501, interpMode=RCIM_Cubic, leaveTangent=0.5501, tangentMode=RCTM_User, time=0, value=0}
          {arriveTangent=1.1924, interpMode=RCIM_Cubic, leaveTangent=1.1924, tangentMode=RCTM_User, time=1, value=0.8}
         ]}}}
      Scale Curve=&RandomRangeFloat(Minimum=-0.3, Maximum=-0.56)
     )
    )
    Index 2 Param 3=&FloatFromCurve(FloatCurve=DataInterface() {propertyValues {Curve {keys [
         {arriveTangent=-0.4045, interpMode=RCIM_Cubic, tangentMode=RCTM_User, time=0.4, value=0.5}
         {arriveTangent=5.0157, interpMode=RCIM_Cubic, leaveTangent=5.0157, tangentMode=RCTM_User, time=1, value=2}
        ]}}})
    Index 2 Param 4=&FloatFromCurve(
     FloatCurve=DataInterface() {propertyValues {Curve {keys [{arriveTangent=-0.4045, time=0, value=0.8}, {leaveTangent=-0.318, time=1, value=1}]}}}
     Scale Curve=0.75
    )
    Index 3 Param 1=&FloatFromCurve(
     FloatCurve=DataInterface() {propertyValues {Curve {keys [
         {arriveTangent=0.1645, interpMode=RCIM_Cubic, leaveTangent=0.1645, tangentMode=RCTM_User, time=0.1, value=0.2}
         {arriveTangent=2.6313, interpMode=RCIM_Cubic, leaveTangent=2.6313, tangentMode=RCTM_User, time=0.6, value=1.25}
         <at 1 2 RCIM_Cubic>
        ]}}}
     Scale Curve=6
    )
    Index 3 Param 2=&FloatFromCurve(
     FloatCurve=DataInterface() {propertyValues {Curve {keys [<at 0 1>,<at 1 0>]}}}
     Scale Curve=&RandomRangeFloat(Minimum=0.4, Maximum=0.5)
    )
    Index 3 Param 3=&FloatFromCurve(
     FloatCurve=DataInterface() {propertyValues {Curve {keys [
         {arriveTangent=0.4884, interpMode=RCIM_Cubic, leaveTangent=0.4884, tangentMode=RCTM_User, time=0, value=0.7}
         {arriveTangent=0.3564, interpMode=RCIM_Cubic, leaveTangent=0.3564, tangentMode=RCTM_User, time=1, value=0.9}
        ]}}}
     Scale Curve=0.0009
    )
    Index 3 Param 4=1
    Module.Index0Param0Write=true
    Module.Index0Param1Write=true
    Module.Index0Param2Write=true
    Module.Index0Param3Write=false
    Module.Index1Param0Write=true
    Module.Index1Param1Write=true
    Module.Index1Param2Write=true
    Module.Index1Param3Write=true
    Module.Index2Param0Write=true
    Module.Index2Param1Write=true
    Module.Index2Param2Write=true
    Module.Index2Param3Write=true
    Module.Index3Param0Write=true
    Module.Index3Param1Write=true
    Module.Index3Param2Write=true
    Module.Index3Param3Write=true
    Write Parameter Index 1=true
    Write Parameter Index 2=true
    Write Parameter Index 3=true
   ) {module=DynamicMaterialParameters001}
  ]
  rendererClasses [&NiagaraSpriteRendererProperties]
  renderers [{rendererClass=&NiagaraSpriteRendererProperties}]
 }
 LW_Debris {rendererClasses [], renderers []}
 LW_FireSplashDrop {rendererClasses [], renderers []}
 LW_Flash {rendererClasses [], renderers []}
 LW_Range_Dome {rendererClasses [], renderers []}
 LW_Sparks1 {rendererClasses [], renderers []}
 LW_Streaks {rendererClasses [], renderers []}
 LW_Streaks001 {rendererClasses [], renderers []}
 Lit {
  emitterSpawnScript [/SetVariables(
    Emitter.CamOfs=25
    Emitter.CameraOffset=5
    Emitter.Color:LinearColor=<rgba 55 4.367 0 5>
    Emitter.Position:NiagaraPosition=<0 0 180>
    Emitter.Position2:NiagaraPosition=<0 0 145>
   ) {isSetParameters=true}]
  emitterUpdateScript [
   &EmitterState(
    Inactive Response=::NewEnumerator1
    Life Cycle Mode=::NewEnumerator1
    Loop Behavior=::NewEnumerator1
    Loop Delay=0.9
    Loop Duration Mode=::NewEnumerator1
    UseLoopDelay=false
   )
   /SetVariables(
    Emitter.DecalSize:Vector3f=&Multiply_VectorByFloat(Vector=<400 600 600>, Float=&MakeCustomFloatFromBool(Bool=$User.IsTrain, True Float=0, False Float=1))
    Emitter.DecalVisible=true
    Emitter.Scale3:Vector3f=&Multiply_VectorByFloat(Vector=<11 11 11>, Float=&MakeCustomFloatFromBool(Bool=$User.IsTrain, True Float=1, False Float=0))
   ) {isSetParameters=true}
   /SetVariables(Emitter.Decal Range Adjustment:Vector3f=&TransformVector(Vector=<0 0 -200>, Destination Space=::World)) {isSetParameters=true}
   /SetVariables(
    Emitter.Color:LinearColor=&MakeLinearColorFromVectorAndFloat("Vector (RGB)"=$Emitter.Decal Range Adjustment, "Float (Alpha)"=$System.Decal_Alpha)
    Emitter.DecalFade=0.1
   ) {isSetParameters=true}
   /SetVariables(
    Emitter.LightColor:LinearColor=&ColorFromCurve(
     ColorCurve=DataInterface() {propertyValues {
       AlphaCurve {keys [<at 0.0069 0>,<at 0.0963 1 RCIM_Cubic>]}
       BlueCurve {keys [<at 0.0115 0>,<at 0.0871 0.0441 RCIM_Cubic>]}
       GreenCurve {keys [<at 0.0115 0>,<at 0.0871 5.9595 RCIM_Cubic>]}
       RedCurve {keys [<at 0.0115 0>,<at 0.0871 75 RCIM_Cubic>]}
      }}
     CurveIndex=$Emitter.Age
    )
    Emitter.LightRadius:NiagaraFloat=&Sine(Normalized Angle=$Emitter.Age, Period=0.2, Scale=5, Bias=180)
   ) {isSetParameters=true}
  ]
  rendererClasses [&NiagaraLightRendererProperties, &NiagaraDecalRendererProperties]
  renderers [
   {rendererClass=&NiagaraLightRendererProperties, rendererIndex=0}
   {rendererClass=&NiagaraDecalRendererProperties, rendererIndex=1}
  ]
 }
 Med_Low {
  emitterUpdateScript [
   &EmitterState(Inactive Response=::NewEnumerator1, Life Cycle Mode=::NewEnumerator1, Loop Delay=0.14, Loop Duration=1, UseLoopDelay=true)
   &SpawnRate(
    Spawn Probability=&MakeCustomFloatFromBool(Bool=$User.IsTrain, True Float=0, False Float=1)
    SpawnRate=9
    Use Spawn Probability=true
   )
  ]
  particleSpawnScript [
   &InitializeParticle(
    Color=<rgba 2 0.2962 0 0.9>
    Lifetime Max=1.3
    Lifetime Mode=::NewEnumerator1
    Mesh Scale Mode=::NewEnumerator3
    Mesh Uniform Scale=&Multiply_Float(A=&RandomRangeFloat(Minimum=0.18, Maximum=0.21), B=0.5)
    Position=$Engine.Owner.Position
    Position Mode=::NewEnumerator1
    Sprite Rotation Angle=&RandomRangeFloat(Minimum=-10, Maximum=10)
    Sprite Rotation Mode=::NewEnumerator0
    Sprite Size Max=<70 50>
    Sprite Size Min=<65 45>
    Sprite Size Mode=::NewEnumerator2
    Sprite UV Mode=::NewEnumerator3
    UsePositionOffset=false
    Write Material Random=true
   ) {module=InitializeParticle001}
   &ShapeLocation(
    Handle Radius=35
    Large Radius=20
    MaskSpawnGroup=false
    Non Uniform Scale=<1 1 0>
    Offset=<0 0 25>
    Offset Coordinate Space=::Local
    Offset Mode=::NewEnumerator0
    Override Seed=false
    Rotation Axis=<0 0 1>
    Rotation Mode=::NewEnumerator1
    Shape Primitive=::NewEnumerator3
    UseOwnerScale=true
   )
   /SetVariables(Particles.SpriteMesh_Quat:Quat4f=&RandomQuaternion(), Particles.Sprite_CamOfs=25) {isSetParameters=true, module=SetVariables_#3}
   &InitialMeshOrientation(
    Mesh Orientation Mode=::NewEnumerator1
    Rotation=&RandomRangeVector(Minimum=<-0.1 -0.1 -1>, Maximum=<0.1 0.1 1>)
    Use Rotation Vector=true
   )
   /SetVariables(
    Particles.InitPos:NiagaraPosition=$Particles.Initial.Position
    Particles.StoredVector:Vector3f=&TransformVector(
     Vector=Linked() {linkedVariable {name=Particles.Initial.Position, type=NiagaraPosition}}
     Source Space=::Simulation
     Destination Space=::World
     Apply Scale=true
     Transform Vector as Position=true
    )
   ) {isSetParameters=true, module=SetVariables_#4}
   &AddVelocity(
    Module.Use Falloff Distance=false
    Origin Offset=&MakeVector(X=0, Y=0, Z=-6)
    Override Seed=false
    Velocity Mode=::NewEnumerator1
    Velocity Origin=$Engine.Owner.Position
    Velocity Speed=&RandomRangeFloat(Minimum=1000, Maximum=2000)
   )
  ]
  particleUpdateScript [
   &ParticleState()
   &UpdateMeshOrientation(Rotation Rate=&RandomRangeFloat(Minimum=-1, Maximum=1), Rotation Vector=<0.15 0.15 0.1>)
   &GravityForce(Gravity=&Multiply_VectorByFloat(
     Vector=<0 0 -15000>
     Float=&MakeCustomFloatFromBool(
      Bool=&SetBoolByFloatComparison(A=15, B=HlslExpression() {hlslExpression="length(Particles.Velocity)"})
      True Float=0
      False Float=1
     )
    ))
   &Collision(
    Age Colliding Particles=false
    CPU Friction Blending:ENiagaraFrictionMergeType=::NewEnumerator3
    Collision Enabled=&SetBoolByFloatComparison(
     A=1
     B=&FloatFromCurve(FloatCurve=DataInterface() {propertyValues {Curve {keys [<at 0.05 0.95>,<at 0.5 2>]}}}, Scale Curve=0.5)
    )
    Friction=10000
    Friction During a Bounce=10000
    Maximum Penetration Correction Distance=1
    Method for Calculating Particle Radius:ENiagaraCalculateRadiusOptions=::NewEnumerator2
    Module.Control Rest State Via Surface Normal=false
    Module.EnableMaxCollisionCount=false
    Module.Randomize Collision Normal Vector=false
    Particle Radius Scale=0.12
    Randomize Collision Normal=0
    Rest State Time Range=1
    Restitution=0
    Simple Friction=false
    Static Friction=1111
    Static Friction Engagement Speed=1111
   )
   &SolveForcesAndVelocity(Clamp Velocity=false, Limit Acceleration=false)
   &ScaleSpriteSize(
    Uniform Curve Scale=3
    Uniform Curve Sprite Scale=DataInterface() {propertyValues {Curve {keys [{tangentMode=RCTM_User, time=0, value=0.2}, <at 0.5 1>, {tangentMode=RCTM_User, time=0.7, value=0}]}}}
   )
   &ScaleMeshSize(Scale Factor=&Multiply_VectorByFloat(
     Vector=&MakeVector(
      X=1
      Y=1
      Z=&ClampFloat(
       Float=&Multiply_Float(
        A=0.01
        B=&DistanceBetweenPositions(A=$Particles.Initial.Position, B=Linked() {linkedVariable {name=Transient.CollisionWorldPosition, type=Vector3f}})
       )
       Min=0.01
       Max=5
      )
     )
     Float=&MakeCustomFloatFromBool(
      Bool=&SetBoolByFloatComparison(A=5, B=HlslExpression() {hlslExpression="length(Particles.Velocity)"})
      True Float=&FloatFromCurve(FloatCurve=DataInterface() {propertyValues {Curve {keys [
           {tangentMode=RCTM_User, time=0, value=0.15}
           <at 0.1 0.5>
           <at 0.1651 0.3>
           <at 0.2226 0.45>
           {tangentMode=RCTM_User, time=0.3, value=0.25}
           <at 0.3288 0.45>
           <at 0.3636 0.4>
           <at 0.4 0.7>
           <at 0.4246 0.4>
           <at 0.4803 0.6>
           <at 0.5186 0.25>
           <at 0.5761 0.4>
           <at 0.6108 0.15>
           <at 0.6533 0.25>
           <at 0.7 0.15>
          ]}}})
      False Float=0
     )
    )) {module=ScaleMeshSize001}
   &DynamicMaterialParameters(
    Index 0 Param 1=&MakeFloatFromVector(Vector=$Particles.StoredVector, Channel=::NewEnumerator0)
    Index 0 Param 2=&MakeFloatFromVector(Vector=$Particles.StoredVector, Channel=::NewEnumerator1)
    Index 0 Param 3=&MakeFloatFromVector(Vector=$Particles.StoredVector, Channel=::NewEnumerator2)
    Index 0 Param 4=&FloatFromCurve(
     FloatCurve=DataInterface() {propertyValues {Curve {keys [{tangentMode=RCTM_User, time=0.025, value=1}, {tangentMode=RCTM_User, time=0.6, value=1}]}}}
     Scale Curve=20
    )
    Module.Param0WriteEnabled=true
    Module.Param1WriteEnabled=true
    Module.Param2WriteEnabled=true
    Module.Param3WriteEnabled=true
   ) {module=DynamicMaterialParameters001}
   &ScaleColor(
    Scale Alpha=&MakeCustomFloatFromBool(
     Bool=&SetBoolByFloatComparison(A=10, B=HlslExpression() {hlslExpression="length(Particles.Velocity)"})
     True Float=&FloatFromCurve(
      FloatCurve=DataInterface() {propertyValues {Curve {keys [
          <at 0.1 0.65 RCIM_Cubic>
          {arriveTangent=-0.0405, interpMode=RCIM_Cubic, leaveTangent=-0.0405, tangentMode=RCTM_User, time=0.2, value=1}
          {interpMode=RCIM_Cubic, leaveTangent=1.75, time=0.4, value=0}
         ]}}}
      Scale Curve=&FNTimeOfDayScalar(Night Scale=0.6, Morning Scale=0.7, Evening Scale=0.7)
     )
     False Float=0
    )
    ScaleA=true
    ScaleRGB=false
   )
   /SetVariables(
    Particles.Color2=&Multiply_LinearColorByFloat(
     Linear Color=&ColorFromCurve(ColorCurve=DataInterface() {propertyValues {
        AlphaCurve {keys [{arriveTangent=1.1089, interpMode=RCIM_Cubic, time=0.107, value=15}, <at 0.9427 1>, <at 0.9661 0 RCIM_Cubic>]}
        BlueCurve {keys [<at 0 0 RCIM_Cubic>]}
        GreenCurve {keys [<at 0 0.3486 RCIM_Cubic>]}
        RedCurve {keys [<at 0 3 RCIM_Cubic>]}
       }})
     Float=1
    )
    Particles.Color3:LinearColor=&ColorFromCurve(ColorCurve=DataInterface() {propertyValues {
       AlphaCurve {keys [<at 0.0514 15 RCIM_Cubic>,<at 0.8816 0 RCIM_Cubic>]}
       BlueCurve {keys [<at 0 0 RCIM_Cubic>]}
       GreenCurve {keys [<at 0 0.1819 RCIM_Cubic>]}
       RedCurve {keys [<at 0 5 RCIM_Cubic>]}
      }})
    Particles.DM2001:Vector4f=&MakeFloat4FromVectorAndFloat("Vector (RGB)"=&TransformVector(Vector=<0 0 0>, Destination Space=::World), "Float (Alpha)"=1)
    Particles.DM2_0:Vector4f=&MakeVector4(
     X=&Multiply_Float(
      A=&FloatFromCurve(
       FloatCurve=DataInterface() {propertyValues {Curve {keys [{tangentMode=RCTM_User, time=0, value=1}, {tangentMode=RCTM_User, time=0.8, value=0}]}}}
       Scale Curve=4
      )
      B=&MakeCustomFloatFromBool(
       Bool=&SetBoolByFloatComparison(A=1, B=HlslExpression() {hlslExpression="length(Particles.Velocity)"})
       True Float=0.2
       False Float=0
      )
     )
     Y=&FloatFromCurve(
      FloatCurve=DataInterface() {propertyValues {Curve {keys [
          {arriveTangent=2.2144, interpMode=RCIM_Cubic, leaveTangent=2.2144, tangentMode=RCTM_User, time=0, value=0}
          {tangentMode=RCTM_User, time=0.9, value=1}
         ]}}}
      Scale Curve=8
     )
     Z=&FloatFromCurve(
      FloatCurve=DataInterface() {propertyValues {Curve {keys [
          {arriveTangent=3.2315, interpMode=RCIM_Cubic, leaveTangent=3.2315, tangentMode=RCTM_User, time=0, value=0}
          {arriveTangent=1.1179, interpMode=RCIM_Cubic, leaveTangent=1.1179, tangentMode=RCTM_User, time=0.4, value=1}
         ]}}}
      Scale Curve=45
     )
     W=&FloatFromCurve(
      FloatCurve=DataInterface() {propertyValues {Curve {keys [
          {arriveTangent=0.6176, interpMode=RCIM_Cubic, leaveTangent=0.6176, tangentMode=RCTM_User, time=0.2, value=0}
          {arriveTangent=6.707, interpMode=RCIM_Cubic, leaveTangent=6.707, tangentMode=RCTM_User, time=1, value=2.5}
         ]}}}
      Scale Curve=-250
     )
    )
    Particles.Mesh Scale2:Vector3f=<10 10 10>
    Particles.SplashSize:Vector2f=&Multiply_Vector2D_ByFloat(
     Float=&FloatFromCurve(FloatCurve=DataInterface() {propertyValues {Curve {keys [{tangentMode=RCTM_User, time=0.1, value=0.02}, {tangentMode=RCTM_User, time=0.8, value=2}]}}})
     Vector2D=<5 5>
    )
    Particles.Vector 2D:Vector2f=&Multiply_Vector2D_ByFloat(
     Float=&MakeCustomFloatFromBool(
      Bool=&SetBoolByFloatComparison(A=15, B=HlslExpression() {hlslExpression="length(Particles.Velocity)"})
      True Float=0
      False Float=1
     )
     Vector2D=&RandomRangeVector2D(Minimum=<5 5>, Maximum=<6 18>)
    )
    Particles.Vector 4:Vector4f=&MakeVector4(
     X=&FloatFromCurve(FloatCurve=DataInterface() {propertyValues {Curve {keys [{arriveTangent=-1.7191, interpMode=RCIM_Cubic, leaveTangent=-1.7191, tangentMode=RCTM_User, time=0, value=1}, <at 0.7 0>]}}})
     Y=0
     Z=0
     W=0
    )
   ) {isSetParameters=true}
   /SetVariables(Particles.Lifetime=&MakeCustomFloatFromBool(
     Bool=&SetBoolByFloatComparison(A=450, B=&DistanceBetweenPositions(A=$Engine.Owner.Position, B=$Particles.Position))
     True Float=$Particles.Lifetime
     False Float=0
    )) {isSetParameters=true}
   &SubUVAnimation(
    Evaluation Type=::NewEnumerator0
    "Sprite Renderer "=DataInterface() {propertyValues {SpriteRenderer=&self:NiagaraEmitter_4.NiagaraSpriteRendererProperties_4}}
    SubUV Animation Mode:ENiagara_SubUVLookupModeV2=::NewEnumerator2
    UseEndFrame=false
    UseStartFrame=false
   )
  ]
  rendererClasses [&NiagaraSpriteRendererProperties, &NiagaraMeshRendererProperties]
  renderers [
   {rendererClass=&NiagaraSpriteRendererProperties}
   {rendererClass=&NiagaraMeshRendererProperties, rendererIndex=1}
   {rendererClass=&NiagaraSpriteRendererProperties, rendererIndex=2}
   {rendererClass=&NiagaraSpriteRendererProperties, rendererIndex=3}
  ]
 }
 TrailEmbers_Hi_Ep {
  simTarget=GPUComputeSim
  emitterUpdateScript [
   &EmitterState(Life Cycle Mode=::NewEnumerator1, Loop Duration=0.15, UseLoopDelay=false)
   &SpawnRate(
    SpawnGroup=1
    SpawnRate=&FloatFromCurve(
     FloatCurve=DataInterface() {propertyValues {Curve {keys [{tangentMode=RCTM_User, time=0, value=0.1}, {tangentMode=RCTM_User, time=0.3, value=1}]}}}
     CurveIndex=$Emitter.Age
     Scale Curve=1000
    )
    Use Spawn Probability=false
   )
  ]
  particleSpawnScript [
   &InitializeParticle(
    Color=&RandomRangeLinearColor(Minimum=<rgba 25 3.0593 0 1>, Maximum=<rgba 25 0.2131 0.0009 1>)
    Lifetime=&MaskFloatBySpawnGroup(Masked Float=0.1, Passthrough Float=&RandomRangeFloat(Minimum=0.7, Maximum=2))
    Sprite Rotation Mode=::NewEnumerator2
    Sprite Size Mode=::NewEnumerator1
    Uniform Sprite Size Max=5
    Uniform Sprite Size Min=2
    UsePositionOffset=false
    Write Material Random=true
   )
   &ShapeLocation(
    MaskSpawnGroup=false
    Non Uniform Scale=<1 1 0.5>
    Offset=<0 0 11>
    Offset Coordinate Space=::Local
    Offset Mode=::NewEnumerator0
    Override Seed=false
    Sphere Radius=25
    UseOwnerScale=true
   )
   &AddVelocity(
    Module.Use Falloff Distance=false
    Origin Offset=<0 0 -25>
    Override Seed=false
    Velocity Mode=::NewEnumerator1
    Velocity Speed=&RandomRangeFloat(Minimum=800, Maximum=2500)
   )
   &CurlNoiseForce(
    Noise Frequency=1
    Noise Strength=50
    Pan Noise=true
    Random Seed:NiagaraInt32=&RandomRangeInt(Minimum=0, Maximum=5000)
   ) {module=CurlNoiseForce001}
   &ApplyInitialForces()
   &CameraOffset(Camera Offset Amount=25) {enabled=false}
  ]
  particleUpdateScript [
   &Drag(
    Drag=&Multiply_Float(
     A=&FloatFromCurve(
      FloatCurve=DataInterface() {propertyValues {Curve {keys [<at 0 0.4>,<at 1 1>]}}}
      Scale Curve=&RandomRangeFloat(Minimum=10, Maximum=15)
     )
     B=&SelectFloatFromArray(
      Direct Array Index=$Particles.PartitionParticles.Partition
      Float Selection Array=DataInterface() {propertyValues {FloatData [1,2,1.5,1.2,1.7]}}
     )
    )
    Use Linear Drag=true
    Use Rotational Drag=false
   )
   &GravityForce(Gravity=&Lerp_Vector(A=<0 0 -700>, B=<0 0 -1500>, Alpha=&RandomRangeFloat(Minimum=0, Maximum=1)))
   &CurlNoiseForce(
    Noise Frequency=5
    Noise Quality / Cost=::NewEnumerator0
    Noise Strength=&FloatFromCurve(
     FloatCurve=DataInterface() {propertyValues {Curve {keys [{interpMode=RCIM_Cubic, tangentMode=RCTM_User, time=0.4, value=1}, {interpMode=RCIM_Cubic, tangentMode=RCTM_User, time=0.7, value=0}]}}}
     Scale Curve=&RandomRangeFloat(Minimum=500, Maximum=1500)
    )
    Pan Noise=false
    Randomization Vector:Vector3f=<0.65 0.125 0.2>
   )
   &Collision(
    Age Colliding Particles=false
    Collision Enabled=&SetBoolByFloatComparison(A=$Particles.NormalizedAge, B=0.05)
    Enable Rest State=false
    Friction=0.2
    GPU Collision Type=::NewEnumerator2
    KillOnCollision=false
    Module.EnableMaxCollisionCount=false
    Module.Randomize Collision Normal Vector=false
    Restitution=&RandomRangeFloat(Minimum=0, Maximum=0.4)
   )
   &SolveForcesAndVelocity(Clamp Velocity=false, Limit Acceleration=false)
   &ParticleState()
   &ScaleColor(
    Scale Alpha=&FloatFromCurve(FloatCurve=DataInterface() {propertyValues {Curve {keys [{interpMode=RCIM_Cubic, tangentMode=RCTM_User, time=-0.5, value=1}, {interpMode=RCIM_Cubic, tangentMode=RCTM_User, time=1, value=0}]}}})
    Scale RGB=&Multiply_VectorByFloat(
     Vector=<3 3 3>
     Float=&FloatFromCurve(
      FloatCurve=DataInterface() {propertyValues {Curve {keys [{interpMode=RCIM_Cubic, tangentMode=RCTM_User, time=0, value=1}, {interpMode=RCIM_Cubic, tangentMode=RCTM_User, time=1, value=0.5}]}}}
      Scale Curve=&Waveform(
       "Waveform [1]"=::NewEnumerator1
       "[1] Phase Scale"=&RandomRangeFloat(Minimum=1, Maximum=2)
       "[1] Frequency"=&Lerp_Float(
        A=&RandomRangeFloat(Minimum=1, Maximum=2)
        B=1
        Alpha=&MaskFloatBySpawnGroup(Masked Float=1, Passthrough Float=1, Spawn Group Mask=1)
       )
       "[1] Amplitude Min/Max"=<0.5 1>
      )
     )
    )
    ScaleA=true
    ScaleRGB=true
   ) {module=ScaleColor001}
   &ScaleSpriteSizeBySpeed(
    Max Scale Factor=<0.15 0.7>
    Min Scale Factor=<0.15 0.2>
    Sample Scale Factor By Curve=true
    Scale Factor Curve=DataInterface() {propertyValues {Curve {keys [
        {interpMode=RCIM_Cubic, tangentMode=RCTM_User, time=0, value=0}
        {arriveTangent=2.9664, interpMode=RCIM_Cubic, leaveTangent=2.9664, tangentMode=RCTM_User, time=1, value=1}
       ]}}}
    Velocity Threshold=500
   )
   &ScaleSpriteSize(Uniform Curve Sprite Scale=DataInterface() {propertyValues {Curve {keys [
        {arriveTangent=2.9868, interpMode=RCIM_Cubic, leaveTangent=2.9868, tangentMode=RCTM_User, time=0, value=1.65}
        <at 0.2 2.25>
        <at 0.9 2.15>
        {arriveTangent=-0.0405, interpMode=RCIM_Cubic, leaveTangent=-0.0405, tangentMode=RCTM_User, time=1, value=1.4}
       ]}}}) {enabled=false}
   &ScaleSpriteSize(
    Scale Sprite Size Mode=::NewEnumerator0
    Uniform Scale Factor=&FN_LOD_Distance_Remap(
     Start Distance=400
     End Distance=8000
     Minimum Output=&RandomRangeFloat(Minimum=1, Maximum=2)
     Maximum Output=5.5
     Clamp Output=true
    )
   ) {module=ScaleSpriteSize001}
  ]
  rendererClasses [&NiagaraSpriteRendererProperties]
  renderers [{rendererClass=&NiagaraSpriteRendererProperties}]
 }
 TrailEmbers_Me_Lo {
  emitterUpdateScript [
   &EmitterState(Life Cycle Mode=::NewEnumerator1, Loop Behavior=::NewEnumerator1, Loop Duration Mode=::NewEnumerator1, UseLoopDelay=false)
   &SpawnRate(
    SpawnRate=&FloatFromCurve(
     FloatCurve=DataInterface() {propertyValues {Curve {keys [{tangentMode=RCTM_User, time=0, value=0.1}, {tangentMode=RCTM_User, time=0.3, value=1}]}}}
     CurveIndex=$Emitter.Age
     Scale Curve=100
    )
    Use Spawn Probability=false
   )
  ]
  particleSpawnScript [
   &InitializeParticle(
    Color=&RandomRangeLinearColor(Minimum=<rgba 25 3.0593 0 1>, Maximum=<rgba 25 0.2131 0.0009 1>)
    Lifetime Max=1
    Lifetime Min=0.8
    Lifetime Mode=::NewEnumerator1
    Sprite Size Mode=::NewEnumerator1
    Uniform Sprite Size Max=6
    UsePositionOffset=false
    Write Material Random=true
   )
   &ShapeLocation(
    MaskSpawnGroup=false
    Non Uniform Scale=<1 1 0.5>
    Offset=<0 0 11>
    Offset Coordinate Space=::Local
    Offset Mode=::NewEnumerator0
    Override Seed=false
    Sphere Radius=25
    UseOwnerScale=true
   )
   &AddVelocity(
    Module.Use Falloff Distance=false
    Origin Offset=<0 0 -20>
    Override Seed=false
    Velocity Mode=::NewEnumerator1
    Velocity Speed=&RandomRangeFloat(Minimum=600, Maximum=2300)
   )
   &CurlNoiseForce(
    Noise Frequency=1
    Noise Strength=50
    Pan Noise=true
    Random Seed:NiagaraInt32=&RandomRangeInt(Minimum=0, Maximum=5000)
   ) {enabled=false, module=CurlNoiseForce001}
   &ApplyInitialForces() {enabled=false}
   &CameraOffset(Camera Offset Amount=25) {enabled=false}
  ]
  particleUpdateScript [
   &Drag(
    Drag=&Multiply_Float(
     A=&FloatFromCurve(
      FloatCurve=DataInterface() {propertyValues {Curve {keys [<at 0 0.4>,<at 1 1>]}}}
      Scale Curve=&RandomRangeFloat(Minimum=10, Maximum=15)
     )
     B=&SelectFloatFromArray(
      Direct Array Index=$Particles.PartitionParticles.Partition
      Float Selection Array=DataInterface() {propertyValues {FloatData [1,2,1.5,1.2,1.7]}}
     )
    )
    Use Linear Drag=true
    Use Rotational Drag=false
   )
   &GravityForce(Gravity=&Lerp_Vector(A=<0 0 -850>, B=<0 0 -1500>, Alpha=&RandomRangeFloat(Minimum=0, Maximum=1)))
   &CurlNoiseForce(
    Noise Frequency=5
    Noise Quality / Cost=::NewEnumerator0
    Noise Strength=&FloatFromCurve(
     FloatCurve=DataInterface() {propertyValues {Curve {keys [{interpMode=RCIM_Cubic, tangentMode=RCTM_User, time=0.4, value=1}, {interpMode=RCIM_Cubic, tangentMode=RCTM_User, time=0.7, value=0}]}}}
     Scale Curve=&RandomRangeFloat(Minimum=500, Maximum=1500)
    )
    Pan Noise=false
    Randomization Vector:Vector3f=<0.65 0.125 0.2>
   )
   &SolveForcesAndVelocity(Clamp Velocity=false, Limit Acceleration=false)
   &ParticleState()
   &ScaleColor(
    Scale Alpha=&FloatFromCurve(FloatCurve=DataInterface() {propertyValues {Curve {keys [{interpMode=RCIM_Cubic, tangentMode=RCTM_User, time=-0.5, value=1}, {interpMode=RCIM_Cubic, tangentMode=RCTM_User, time=1, value=0}]}}})
    Scale RGB=&Multiply_VectorByFloat(
     Vector=<3 3 3>
     Float=&FloatFromCurve(
      FloatCurve=DataInterface() {propertyValues {Curve {keys [
          {interpMode=RCIM_Cubic, tangentMode=RCTM_User, time=0.65, value=1}
          {arriveTangent=-1.236, interpMode=RCIM_Cubic, leaveTangent=-1.236, tangentMode=RCTM_User, time=1, value=0.5}
         ]}}}
      Scale Curve=&Waveform(
       "Waveform [1]"=::NewEnumerator1
       "[1] Phase Scale"=&RandomRangeFloat(Minimum=1, Maximum=2)
       "[1] Frequency"=&Lerp_Float(
        A=&RandomRangeFloat(Minimum=1, Maximum=2)
        B=1
        Alpha=&MaskFloatBySpawnGroup(Masked Float=1, Passthrough Float=1, Spawn Group Mask=1)
       )
       "[1] Amplitude Min/Max"=<0.5 1>
      )
     )
    )
    ScaleA=true
    ScaleRGB=true
   ) {module=ScaleColor001}
   &ScaleSpriteSizeBySpeed(
    Max Scale Factor=<0.15 0.7>
    Min Scale Factor=<0.15 0.2>
    Sample Scale Factor By Curve=true
    Scale Factor Curve=DataInterface() {propertyValues {Curve {keys [
        {interpMode=RCIM_Cubic, tangentMode=RCTM_User, time=0, value=0}
        {arriveTangent=2.9664, interpMode=RCIM_Cubic, leaveTangent=2.9664, tangentMode=RCTM_User, time=1, value=1}
       ]}}}
    Velocity Threshold=500
   )
   &ScaleSpriteSize(
    Scale Sprite Size Mode=::NewEnumerator0
    Uniform Scale Factor=&FN_LOD_Distance_Remap(
     Start Distance=400
     End Distance=8000
     Minimum Output=&RandomRangeFloat(Minimum=2, Maximum=5)
     Maximum Output=5.5
     Clamp Output=true
    )
   ) {module=ScaleSpriteSize001}
  ]
  rendererClasses [&NiagaraSpriteRendererProperties]
  renderers [{rendererClass=&NiagaraSpriteRendererProperties}]
 }
}
summary {
 systemName=NS_Thermite_AoE
 emitters [
  {emitterName=LW_Sparks1, rendererClasses []}
  {emitterName=LW_Debris, rendererClasses []}
  {emitterName=LW_Flash, rendererClasses []}
  {emitterName=LW_Streaks, rendererClasses []}
  {emitterName=TrailEmbers_Hi_Ep, simTarget=GPUComputeSim, rendererClasses [&NiagaraSpriteRendererProperties]}
  {emitterName=Lit, rendererClasses [&NiagaraLightRendererProperties, &NiagaraDecalRendererProperties]}
  {
   emitterName=Epic_High
   simTarget=GPUComputeSim
   rendererClasses [&NiagaraSpriteRendererProperties, &NiagaraMeshRendererProperties]
  }
  {emitterName=Med_Low, rendererClasses [&NiagaraSpriteRendererProperties, &NiagaraMeshRendererProperties]}
  {emitterName=Fire_Smoke, rendererClasses [&NiagaraSpriteRendererProperties]}
  {emitterName=TrailEmbers_Me_Lo, rendererClasses [&NiagaraSpriteRendererProperties]}
  {emitterName=LW_Range_Dome, rendererClasses []}
  {emitterName=LW_Streaks001, rendererClasses []}
  {emitterName=LW_FireSplashDrop, rendererClasses []}
 ]
 userVariables [
  User.Duration {type=NiagaraFloat, defaultValue=15, description=""}
  User.IsTrain {type=NiagaraBool}
  User.Radius {type=NiagaraFloat, defaultValue=300, description=""}
 ]
}
