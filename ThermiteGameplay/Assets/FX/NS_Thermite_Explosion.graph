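# Niagara system for a thermite explosion with smoke, fire, embers, debris, sparks and flash.
# A burst of GPU SourceParticles feeds temperature and velocity into a camera-facing Grid2D Niagara Fluids gas simulation with lighting grids, limited by instance-count culling.
# A GPU ember burst uses drag, curl noise and collision, while Lit001 drives a timed orange light and fading decal.
# User.Global Scale sizes the fluid effect, and User.IsTrain suppresses the decal.
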NiagaraSystem /ThermiteGameplay/Assets/FX/NS_Thermite_Explosion
dependencies {
 usedDataInterfaces [
  &NiagaraDataInterfaceCurve
  &NiagaraDataInterfaceArrayFloat
  &NiagaraDataInterfaceColorCurve
  &NiagaraDataInterfacePlatformSet
  &NiagaraDataInterfaceCamera
  /Script/Niagara.NiagaraDataInterfaceGrid2DCollection
  /Script/Niagara.NiagaraDataInterfaceRenderTarget2D
  &NiagaraDataInterfaceParticleRead
  /Script/Niagara.NiagaraDataInterfaceEmitterProperties
 ]
 usedDynamicInputs [&MakeVector]
 usedModules [
  &self:SystemSpawnScript.SystemScriptSource.SystemScriptGraph.NiagaraNodeAssignment_0.SetVariables_#1
  &SystemState
  &self:SystemSpawnScript.SystemScriptSource.SystemScriptGraph.NiagaraNodeAssignment_1.SetVariables_#2
 ]
}
emitters {
 Dark_Chunks {rendererClasses [], renderers []}
 Grid2D_Gas_SmokeFire_Emitter {
  simTarget=GPUComputeSim
  emitterSpawnScript [
   /NiagaraFluids/Functions/General/Fluids_Gas_Manager(
    Density Bouyancy=-0.1
    Density Color Curve=DataInterface() {propertyValues {
      ExposedName=Curve Density
      bExposeCurve=true
      AlphaCurve {keys [<at 0 0 RCIM_Cubic>,<at 2.7618 1>]}
      BlueCurve {keys [<at 0.0245 0.045>,<at 4.4467 0.055>]}
      GreenCurve {keys [<at 0.0245 0.0307>,<at 4.4467 0.0375>]}
      RedCurve {keys [<at 0.0245 0.0266>,<at 4.4467 0.0325>]}
     }}
    Depth Fade=25
    Dissipation Temperature=5
    Dissipation Velocity=0.5
    EnvironmentColor:LinearColor=<rgba 0.1427 0.2031 0.3542 1>
    Pressure Relaxation=1
    Radius Scale=0.8
    Shadowing=0.25
    SunColor:LinearColor=<rgba 1.2417 1.53 2 1>
    Temperature Bouyancy=0.0025
    Temperature Color Curve=DataInterface() {propertyValues {
      ExposedName=Curve Temperature
      bExposeCurve=true
      AlphaCurve {keys [<at 0 0>,<at 0.4853 1>]}
      BlueCurve {keys [<at 0.0696 0>,<at 0.8304 0.0517>,<at 0.9917 0.0783>]}
      GreenCurve {keys [<at 0.0696 0>,<at 0.8304 0.729>,<at 0.9917 0.2349>]}
      RedCurve {keys [<at 0.0696 0>,<at 0.8304 8>,<at 0.9917 0.9401>]}
     }}
    Temperature Maximum=1280
    Temperature Scale=0.2
    Vorticity Confinement=0.1
   )
   /SetVariables(Emitter.SimGrid=DataInterface() {propertyValues {
      ClearBeforeNonIterationStage=true
      NumAttributes=0
      NumCellsMaxAxis=3
      NumCellsX=1
      NumCellsY=1
      OverrideBufferFormat=Float
      PreviewAttribute=None
      SetGridFromMaxAxis=false
      bOverrideFormat=false
      bPreviewGrid=false
      RenderTargetUserParameter {parameter {name=None, typeDefHandle {registeredTypeIndex=1}, varData []}}
      WorldBBoxSize=<1200 1200>
     }}) {isSetParameters=true}
   /SetVariables(Emitter.LightingGrid=DataInterface() {propertyValues {
      ClearBeforeNonIterationStage=true
      NumAttributes=3
      NumCellsMaxAxis=3
      NumCellsX=512
      NumCellsY=512
      OverrideBufferFormat=Float
      PreviewAttribute=None
      SetGridFromMaxAxis=false
      bOverrideFormat=false
      bPreviewGrid=false
      RenderTargetUserParameter {parameter {name=None, typeDefHandle {registeredTypeIndex=4}, varData []}}
      WorldBBoxSize=<100 100>
     }}) {isSetParameters=true}
   /SetVariables(Emitter.LightingGridDownsample=DataInterface() {propertyValues {
      ClearBeforeNonIterationStage=true
      NumAttributes=1
      NumCellsMaxAxis=3
      NumCellsX=64
      NumCellsY=64
      OverrideBufferFormat=Float
      PreviewAttribute=None
      SetGridFromMaxAxis=false
      bOverrideFormat=false
      bPreviewGrid=false
      RenderTargetUserParameter {parameter {name=None, typeDefHandle {registeredTypeIndex=4}, varData []}}
      WorldBBoxSize=<100 100>
     }}) {isSetParameters=true}
   /SetVariables() {isSetParameters=true}
   /SetVariables() {isSetParameters=true}
   /SetVariables() {isSetParameters=true}
   /NiagaraFluids/Modules/Grid2D/Grid2D_SetResolution(
    Grid:NiagaraDataInterfaceGrid2DCollection=$Emitter.SimGrid
    NumCellsMaxAxis=320
    Resolution Method:ENiagaraGrid2DResolution=::NewEnumerator1
    WorldGridExtents:Vector2f=&Multiply_Vector2D_ByFloat(Float=$User.Global Scale, Vector2D=<1280 1280>)
   )
   &Grid2D_SetResolution(
    Grid:NiagaraDataInterfaceGrid2DCollection=$Emitter.LightingGridDownsample
    OtherGrid:NiagaraDataInterfaceGrid2DCollection=$Emitter.SimGrid
    Resolution Method:ENiagaraGrid2DResolution=::NewEnumerator3
    Resolution Mult=0.2
    WorldGridExtents:Vector2f=$Emitter.Grid2D_SetResolution.WorldGridExtents
   ) {module=Grid2D_SetResolution003}
   &Grid2D_SetResolution(
    Grid:NiagaraDataInterfaceGrid2DCollection=$Emitter.LightingGrid
    OtherGrid:NiagaraDataInterfaceGrid2DCollection=$Emitter.SimGrid
    Resolution Method:ENiagaraGrid2DResolution=::NewEnumerator3
    WorldGridExtents:Vector2f=$Emitter.Grid2D_SetResolution.WorldGridExtents
   ) {module=Grid2D_SetResolution004}
   &Grid2D_SetResolution(
    Grid:NiagaraDataInterfaceGrid2DCollection=$Emitter.SourceGrid
    OtherGrid:NiagaraDataInterfaceGrid2DCollection=$Emitter.SimGrid
    Resolution Method:ENiagaraGrid2DResolution=::NewEnumerator3
    WorldGridExtents:Vector2f=$Emitter.Grid2D_SetResolution.WorldGridExtents
   ) {module=Grid2D_SetResolution005}
   /SetVariables(Emitter.g:Vector3f=<0 0 -980>) {isSetParameters=true}
   &Grid2D_SetResolution(
    Grid:NiagaraDataInterfaceGrid2DCollection=$Emitter.PressureGrid
    OtherGrid:NiagaraDataInterfaceGrid2DCollection=$Emitter.SimGrid
    Resolution Method:ENiagaraGrid2DResolution=::NewEnumerator3
    WorldGridExtents:Vector2f=$Emitter.Grid2D_SetResolution.WorldGridExtents
   ) {module=Grid2D_SetResolution001}
   &Grid2D_SetResolution(
    Grid:NiagaraDataInterfaceGrid2DCollection=$Emitter.TemporaryGrid
    OtherGrid:NiagaraDataInterfaceGrid2DCollection=$Emitter.SimGrid
    Resolution Method:ENiagaraGrid2DResolution=::NewEnumerator3
    WorldGridExtents:Vector2f=$Emitter.Grid2D_SetResolution.WorldGridExtents
   ) {module=Grid2D_SetResolution002}
   /SetVariables() {isSetParameters=true}
   /SetVariables(Emitter.dt:NiagaraFloat=$Engine.DeltaTime) {enabled=false, isSetParameters=true}
   /SetVariables(Emitter.ParticleSource=DataInterface() {propertyValues {EmitterBinding {bindingMode=Other, emitterName=SourceParticles}}}) {isSetParameters=true}
   /SetVariables() {isSetParameters=true}
   &self:NiagaraEmitter_1.NiagaraScratchPadContainer_0.SetLightingRT_0(LightingGrid:NiagaraDataInterfaceGrid2DCollection=$Emitter.LightingGrid, LightingRT:NiagaraDataInterfaceRenderTarget2D=$Emitter.LightingRT) {module=SetLightingRT_29}
   /SetVariables() {enabled=false, isSetParameters=true}
   /SetVariables(Emitter.density=1) {isSetParameters=true}
   /NiagaraFluids/Modules/Grid2D/Grid2D_UpdateSimAttributes(Grid:NiagaraDataInterfaceGrid2DCollection=$Emitter.SimGrid, WorldGridSize:Vector2f=$Emitter.Grid2D_SetResolution.WorldGridExtents) {module=UpdateSimAttributes}
   /SetVariables() {isSetParameters=true}
   &self:NiagaraEmitter_1.NiagaraScratchPadContainer_0.ScratchModule_02_1(New Grid2D Collection:NiagaraDataInterfaceGrid2DCollection=$Emitter.SimGrid) {module=ScratchModule_02001}
   /SetVariables(Emitter.PressureIterations=5) {isSetParameters=true}
   /NiagaraFluids/Modules/Grid2D/Grid2D_InitRT(Grid:NiagaraDataInterfaceGrid2DCollection=$Emitter.SimGrid, SimRT:NiagaraDataInterfaceRenderTarget2D=$Emitter.SimRT) {module=Grid2D_Gas_InitRT}
   &self:NiagaraEmitter_1.NiagaraScratchPadContainer_0.ScratchModule_0(
    Grid:NiagaraDataInterfaceGrid2DCollection=$Emitter.SimGrid
    PressureGrid:NiagaraDataInterfaceGrid2DCollection=$Emitter.PressureGrid
    TemporaryGrid:NiagaraDataInterfaceGrid2DCollection=$Emitter.TemporaryGrid
   ) {module=ScratchModule}
   /SetVariables(Emitter.GridsAllocated=true) {isSetParameters=true}
  ]
  emitterUpdateScript [
   &self:Emitter_Limitation_and_Culling__01(Max Num=3, Platform Set=DataInterface() {propertyValues {Platforms {qualityLevelMask=-17, cVarConditions [], deviceProfileStates []}}}) {module=Emitter_Limitation_and_Culling__01}
   &EmitterState(Life Cycle Mode=::NewEnumerator1, Loop Behavior=::NewEnumerator1, Loop Duration=3.6, Scalability Mode=::NewEnumerator1, UseLoopDelay=false)
   /SetVariables(Emitter.dt:NiagaraFloat=$Engine.DeltaTime) {isSetParameters=true}
   /NiagaraFluids/Modules/Grid2D/Grid2D_CreateUnitToWorldTransform(
    Camera:NiagaraDataInterfaceCamera=$Emitter.CameraQuery
    Camera Facing=true
    Offset:NiagaraPosition=$Engine.Owner.Position
    Rotation:Quat4f=$Engine.Owner.Rotation
    WorldGridExtents:Vector2f=$Emitter.Grid2D_SetResolution.WorldGridExtents
   )
   /SetVariables(Emitter.Position:NiagaraPosition=&ConvertVectorToPosition(
     Input Position=$Emitter.Grid2D_CreateUnitToWorldTransform.LocalToWorldTranslation
     Conversion Operation:ENiagara_LWCConvertVecToPos=::NewEnumerator0
    )) {isSetParameters=true}
   /NiagaraFluids/Modules/Grid2D/Grid2D_SetBounds(
    CameraFacing:NiagaraBool=$Emitter.Grid2D_CreateUnitToWorldTransform.CameraFacing
    EmitterProperties=DataInterface() {propertyValues {EmitterBinding {bindingMode=Self, emitterName=None}}}
    LocalPivot:Vector2f=$Emitter.Grid2D_CreateUnitToWorldTransform.LocalPivot
    Rotate for Camera Facing=true
    WorldGridExtents:Vector2f=$Emitter.Grid2D_SetResolution.WorldGridExtents
   )
   /SetVariables(Emitter.ExecutionState:ENiagaraExecutionState=$Emitter.ENiagaraExecutionState) {isSetParameters=true}
   &self:ScratchModule_02(PerformDeallocation:NiagaraBool=/Niagara/DynamicInputs/Scalability/IsEmitterComplete()) {module=ScratchModule_02002}
  ]
  rendererClasses [&NiagaraSpriteRendererProperties]
  renderers [
   {rendererClass=&NiagaraSpriteRendererProperties}
   {rendererClass=&NiagaraSpriteRendererProperties, rendererIndex=1}
  ]
 }
 LW_Debris {rendererClasses [], renderers []}
 LW_Explosion_Main {rendererClasses [], renderers []}
 LW_Fire {rendererClasses [], renderers []}
 LW_Flash {rendererClasses [], renderers []}
 LW_GlowSphereMesh2 {rendererClasses [], renderers []}
 LW_Sparks1 {rendererClasses [], renderers []}
 LW_Streaks {rendererClasses [], renderers []}
 Lit001 {
  emitterSpawnScript [/SetVariables(Emitter.CamOfs=25, Emitter.Position2:NiagaraPosition=<0 0 200>) {isSetParameters=true}]
  emitterUpdateScript [
   /SetVariables(
    Emitter.Decal Range Adjustment:Vector3f=&TransformVector(Vector=<1 1 2>, Destination Space=::World)
    Emitter.DecalSize:Vector3f=&Multiply_VectorByFloat(Vector=<150 600 600>, Float=&MakeCustomFloatFromBool(Bool=$User.IsTrain, True Float=0, False Float=1))
    Emitter.DecalVisible:NiagaraBool=&Boolean_NotOperation(Bool=$User.IsTrain)
   ) {isSetParameters=true}
   &EmitterState(
    Inactive Response=::NewEnumerator1
    Life Cycle Mode=::NewEnumerator1
    Loop Behavior=::NewEnumerator1
    Loop Delay=0.9
    Loop Duration=4
    UseLoopDelay=false
   )
   /SetVariables(
    Emitter.Color:LinearColor=&MakeLinearColorFromVectorAndFloat(
     "Vector (RGB)"=$Emitter.Decal Range Adjustment
     "Float (Alpha)"=&FloatFromCurve(
      FloatCurve=DataInterface() {propertyValues {Curve {keys [<at 0.1 0.55>, {interpMode=RCIM_Cubic, leaveTangent=-1.7678, tangentMode=RCTM_Break, time=0.4, value=0.55}, <at 1 0>]}}}
      CurveIndex=$Emitter.NormalizedLoopAge
      Scale Curve=0.1
     )
    )
    Emitter.DecalFade:NiagaraFloat=&FloatFromCurve(
     FloatCurve=DataInterface() {propertyValues {Curve {keys [
         {arriveTangent=0.0746, interpMode=RCIM_Cubic, leaveTangent=0.0746, tangentMode=RCTM_User, time=0.2, value=0}
         {arriveTangent=2.686, interpMode=RCIM_Cubic, leaveTangent=2.686, tangentMode=RCTM_User, time=0.6, value=0.4}
         {arriveTangent=0.158, interpMode=RCIM_Cubic, leaveTangent=0.158, tangentMode=RCTM_User, time=1, value=1}
        ]}}}
     CurveIndex=$Emitter.NormalizedLoopAge
    )
   ) {isSetParameters=true}
   /SetVariables(
    Emitter.LightColor:LinearColor=&ColorFromCurve(
     ColorCurve=DataInterface() {propertyValues {
       AlphaCurve {keys [<at 0.0069 0>,<at 0.0963 1 RCIM_Cubic>,<at 0.7 0>]}
       BlueCurve {keys [<at 0.0115 0>, {arriveTangent=0.1081, interpMode=RCIM_Cubic, leaveTangent=0.1081, time=0.0871, value=0.0588}, <at 0.5238 0.0554>, <at 0.7 0>]}
       GreenCurve {keys [
         <at 0.0115 0>
         {arriveTangent=14.6025, interpMode=RCIM_Cubic, leaveTangent=14.6025, time=0.0871, value=7.946}
         <at 0.5238 7.4813>
         <at 0.7 0>
        ]}
       RedCurve {keys [
         <at 0.0115 0>
         {arriveTangent=183.7726, interpMode=RCIM_Cubic, leaveTangent=183.7726, time=0.0871, value=100}
         <at 0.5238 94.1525>
         <at 0.7 0>
        ]}
      }}
     CurveIndex=$Emitter.Age
    )
    Emitter.LightEnabled:NiagaraBool=&SetBoolByFloatComparison(
     A=0.5
     B=&FloatFromCurve(FloatCurve=DataInterface() {propertyValues {Curve {keys [<at 0 1 RCIM_Constant>,<at 0.7 -1>]}}}, CurveIndex=$Emitter.Age)
     Comparison Type=::NewEnumerator6
    )
    Emitter.LightRadius=400
   ) {isSetParameters=true}
  ]
  rendererClasses [&NiagaraLightRendererProperties, &NiagaraDecalRendererProperties]
  renderers [
   {rendererClass=&NiagaraLightRendererProperties, rendererIndex=0}
   {rendererClass=&NiagaraDecalRendererProperties, rendererIndex=1}
  ]
 }
 SourceParticles {
  simTarget=GPUComputeSim
  emitterUpdateScript [
   /NiagaraFN/Modules/Scalability/SignificantInstancesCull(
    Cull Response:ENiagaraExecutionStateManagement=::KillImmediately
    First Frame Only=false
    Num Instances=3
    Platform Set=DataInterface() {propertyValues {Platforms {qualityLevelMask=-17, cVarConditions [], deviceProfileStates []}}}
   )
   &EmitterState(
    Life Cycle Mode=::NewEnumerator1
    Loop Behavior=::NewEnumerator1
    Loop Delay=0.015
    Loop Duration=0.2
    Scalability Mode=::NewEnumerator1
    UseLoopDelay=true
   )
   &SpawnBurst_Instantaneous(
    Spawn Count=&RandomRangeInt(Minimum=6, Maximum=10, Evaluation Type=::NewEnumerator1)
    Use Spawn Probability=false
   )
   &SpawnRate(SpawnRate=55, Use Spawn Probability=false) {enabled=false}
   &Grid2D_CreateUnitToWorldTransform(
    Camera Facing=true
    Local Pivot:Vector2f=<0.5 0.1>
    Offset:NiagaraPosition=$Engine.Owner.Position
    Rotation:Quat4f=$Engine.Owner.Rotation
    WorldGridExtents:Vector2f=$Emitter.Grid2D_SetResolution.WorldGridExtents
   ) {enabled=false}
  ]
  particleSpawnScript [
   /Niagara/Modules/Spawn/Initialization/InitializeParticle(
    Color=<rgba 1111 13.6009 0 1>
    Lifetime=&RandomRangeFloat(Minimum=0.85, Maximum=1.5)
    Mass=&RandomRangeFloat(Minimum=0.75, Maximum=1)
    Module.Write Color=true
    Module.Write Lifetime=true
    Module.Write Mass=true
    Module.Write Material Random=true
    Module.Write Position=true
    Module.Write Scale=false
    Module.Write SpriteRotation=false
    Module.Write SpriteSize=true
    Sprite Size=&Vector2DFromFloat(Value=1)
   )
   &AddVelocityInCone(
    Cone Angle=80
    Cone Axis=<0 0 1>
    Cone Axis Coordinate Space=::World
    Module.Use Velocity Falloff On Cone Axis=true
    Velocity Distribution Along Cone Axis=0
    Velocity Falloff Away From Cone Axis=0
    Velocity Strength=&Multiply_Float(A=&RandomRangeFloat(Minimum=1075, Maximum=1756), B=$User.Global Scale)
   )
   &SphereLocation(
    MaskSpawnGroup=false
    Override Seed=false
    Sphere Radius=&Multiply_Float(
     A=&Lerp_Float(A=25, B=50, Alpha=&Sine(Normalized Angle=$Emitter.Age, Period=4, Scale=0.5, Bias=1))
     B=$User.Global Scale
    )
    Surface Only=false
    UseNonUniformScale=true
    UseOffset=true
    UseSphereOrientation=false
   ) {enabled=false}
   /Niagara/Modules/Update/Forces/CurlNoiseForce(
    Noise Frequency:NiagaraFloat=&Divide_Float(A=0.04, B=$User.Global Scale)
    Noise Strength=&Multiply_Float(A=250, B=$User.Global Scale)
    Pan Noise=true
    Pan Noise Field=<0.3 0.5 0.3>
   ) {enabled=false}
   &InheritVelocity(
    Module.Limit Speed=true
    Velocity Limit:NiagaraFloat=&Multiply_Float(A=1500, B=$User.Global Scale)
    Velocity Scale=<0 0 0>
   )
   &ApplyInitialForces()
   &TorusLocation(
    Align Torus With Axis=true
    Large Radius:NiagaraFloat=&Multiply_Float(A=60, B=$User.Global Scale)
    MaskSpawnGroup=false
    Torus Axis=<1 0 0>
    Torus Distribution Mode:ENiagaraTorusDistributionMode=::NewEnumerator1
    Torus Mode=::NewEnumerator2
    U Position:NiagaraFloat=&Subtract_Float(A=&ReturnNormalizedExecIndex(), B=2)
   )
   &AddVelocityFromPoint(
    Module.Use Falloff Distance=false
    Origin Offset=&Multiply_VectorByFloat(
     Vector=$Engine.Owner.SystemXAxis
     Float=&RandomRangeFloat(Minimum=0, Maximum=&Multiply_Float(A=-60, B=$User.Global Scale))
    )
    Origin Offset Coordinate Space=::Simulation
    Velocity Origin:Vector3f=&SimulationPosition()
    Velocity Strength=&Multiply_Float(A=&UniformRangedFloat(Minimum=5, Maximum=10), B=$User.Global Scale)
   ) {module=AddVelocityFromPoint001}
  ]
  particleUpdateScript [
   &ParticleState()
   &CurlNoiseForce(Noise Strength=&Multiply_Float(A=300, B=$User.Global Scale), Pan Noise=true, Pan Noise Field=<0.1 0.5 0.2>) {module=CurlNoiseForce001}
   &GravityForce(Gravity=&Multiply_VectorByFloat(Vector=<0 0 -1500>, Float=$User.Global Scale))
   &Drag(Drag=&Multiply_Float(A=1.4, B=1), Use Linear Drag=true, Use Rotational Drag=false)
   &ScaleColor(
    Scale Alpha=&FloatFromCurve(FloatCurve=DataInterface() {propertyValues {Curve {keys [<at -0.0046 1.0113 RCIM_Cubic>,<at 1.0011 0.002 RCIM_Cubic>]}}})
    ScaleA=true
    ScaleRGB=true
   )
   &Collision(
    Age Colliding Particles=false
    Collision Enabled=&SetBoolByFloatComparison(A=$Particles.Age, B=0.4)
    Enable Rest State=false
    Friction=0.2
    Friction During a Bounce=0.3
    Kill Offscreen Particles=true
    Maximum time offscreen=99
    Module.EnableMaxCollisionCount=false
    Module.Randomize Collision Normal Vector=false
    Restitution=0.2
   ) {module=Collision001}
   &SolveForcesAndVelocity(Clamp Velocity=false, Limit Acceleration=false)
   /NiagaraFluids/Modules/Particles/Fluids_Gas_Source(
    NoiseIntensity=3
    NoiseScale=3
    Radius:NiagaraFloat=&FloatFromCurve(
     FloatCurve=DataInterface() {propertyValues {Curve {keys [<at 0 0>, {interpMode=RCIM_Cubic, leaveTangent=-2.282, tangentMode=RCTM_Break, time=0.0361, value=0.984}, <at 1.0022 0.3>]}}}
     Scale Curve=&Multiply_Float(A=&RandomRangeFloat(Minimum=30, Maximum=40), B=$User.Global Scale)
    )
    Temperature:NiagaraFloat=&FloatFromCurve(
     FloatCurve=DataInterface() {propertyValues {Curve {keys [<at 0 1>,<at 0.228 0.772>,<at 1 0>]}}}
     Scale Curve=&RandomRangeFloat(Minimum=500, Maximum=3000)
    )
    VelocityScale:NiagaraFloat=&RandomRangeFloat(Minimum=0.5, Maximum=1)
   )
   /SetVariables(Particles.SpriteSize=&Vector2DFromFloat(Value=&Multiply_Float(A=$Particles.fluids_source_radius, B=&Multiply_Float(A=1.4, B=1.1)))) {enabled=false, isSetParameters=true}
  ]
  rendererClasses [&NiagaraSpriteRendererProperties]
  renderers [{rendererClass=&NiagaraSpriteRendererProperties}]
 }
 TrailEmbers_Hi_Ep001 {
  simTarget=GPUComputeSim
  emitterUpdateScript [
   &EmitterState(Life Cycle Mode=::NewEnumerator1, Loop Behavior=::NewEnumerator1, Loop Duration=0.15, UseLoopDelay=false)
   &SpawnBurst_Instantaneous(Spawn Count=400, Spawn Group=1, Use Spawn Probability=false, UseLoopCountLimit=false)
  ]
  particleSpawnScript [
   &PartitionParticles(Number Of Partitions=1) {enabled=false}
   /Niagara/Modules/Spawn/Initialization/V2/InitializeParticle(
    Color=&RandomRangeLinearColor(Minimum=<rgba 25 3.0593 0 1>, Maximum=<rgba 25 0.2131 0.0009 1>)
    Lifetime=&MaskFloatBySpawnGroup(Masked Float=0.1, Passthrough Float=&RandomRangeFloat(Minimum=0.7, Maximum=2))
    Sprite Rotation Mode=::NewEnumerator2
    Sprite Size Mode=::NewEnumerator1
    Uniform Sprite Size Max=5
    Uniform Sprite Size Min=2
    UsePositionOffset=false
    Write Material Random=true
   )
   &AddVelocity(
    Module.Use Falloff Distance=false
    Override Seed=false
    Velocity Mode=::NewEnumerator1
    Velocity Speed=&RandomRangeFloat(Minimum=800, Maximum=3000)
   )
   &ShapeLocation(
    MaskSpawnGroup=false
    Non Uniform Scale=<0.01 1 1>
    Offset=<10 0 0>
    Offset Coordinate Space=::Local
    Offset Mode=::NewEnumerator0
    Override Seed=false
    UseOwnerScale=true
   )
   /Niagara/Modules/Update/Forces/V2/CurlNoiseForce(
    Noise Frequency=1
    Noise Strength=50
    Pan Noise=true
    Random Seed:NiagaraInt32=&RandomRangeInt(Minimum=0, Maximum=5000)
   ) {module=CurlNoiseForce001}
   &ApplyInitialForces()
   &CameraOffset(Camera Offset Amount=75)
  ]
  particleUpdateScript [
   &Drag(
    Drag=&Multiply_Float(
     A=&FloatFromCurve(
      FloatCurve=DataInterface() {propertyValues {Curve {keys [<at 0 0.4>,<at 1 1>]}}}
      Scale Curve=&RandomRangeFloat(Minimum=10, Maximum=15)
     )
     B=&SelectFloatFromArray(
      Direct Array Index=$Particles.PartitionParticles.Partition
      Float Selection Array=DataInterface() {propertyValues {FloatData [1,2,1.5,1.2,1.7]}}
     )
    )
    Use Linear Drag=true
    Use Rotational Drag=false
   )
   &GravityForce(Gravity=&Lerp_Vector(A=<0 0 -700>, B=<0 0 -1500>, Alpha=&RandomRangeFloat(Minimum=0, Maximum=1)))
   &CurlNoiseForce(
    Noise Frequency=5
    Noise Quality / Cost=::NewEnumerator0
    Noise Strength=&FloatFromCurve(
     FloatCurve=DataInterface() {propertyValues {Curve {keys [{interpMode=RCIM_Cubic, tangentMode=RCTM_User, time=0.4, value=1}, {interpMode=RCIM_Cubic, tangentMode=RCTM_User, time=0.7, value=0}]}}}
     Scale Curve=&RandomRangeFloat(Minimum=500, Maximum=1500)
    )
    Pan Noise=false
    Randomization Vector:Vector3f=<0.65 0.125 0.2>
   )
   &Collision(
    Age Colliding Particles=false
    Collision Enabled=&SetBoolByFloatComparison(A=$Particles.NormalizedAge, B=0.05)
    Enable Rest State=false
    Friction=0.2
    GPU Collision Type=::NewEnumerator2
    KillOnCollision=false
    Module.EnableMaxCollisionCount=false
    Module.Randomize Collision Normal Vector=false
    Restitution=&RandomRangeFloat(Minimum=0, Maximum=0.4)
   )
   &SolveForcesAndVelocity(Clamp Velocity=false, Limit Acceleration=false)
   &ParticleState()
   &DynamicMaterialParameters(
    Index 0 Param 1=0.8
    Index 0 Param 2=1
    Index 0 Param 3=1
    Module.Param0WriteEnabled=true
    Module.Param1WriteEnabled=false
    Module.Param2WriteEnabled=false
    Module.Param3WriteEnabled=false
   )
   &ScaleColor(
    Scale Alpha=&FloatFromCurve(FloatCurve=DataInterface() {propertyValues {Curve {keys [{interpMode=RCIM_Cubic, tangentMode=RCTM_User, time=-0.5, value=1}, {interpMode=RCIM_Cubic, tangentMode=RCTM_User, time=1, value=0}]}}})
    Scale RGB=&Multiply_VectorByFloat(
     Vector=<3 3 3>
     Float=&FloatFromCurve(
      FloatCurve=DataInterface() {propertyValues {Curve {keys [{interpMode=RCIM_Cubic, tangentMode=RCTM_User, time=0, value=1}, {interpMode=RCIM_Cubic, tangentMode=RCTM_User, time=1, value=0.5}]}}}
      Scale Curve=&Waveform(
       "Waveform [1]"=::NewEnumerator1
       "[1] Phase Scale"=&RandomRangeFloat(Minimum=1, Maximum=2)
       "[1] Frequency"=&Lerp_Float(
        A=&RandomRangeFloat(Minimum=1, Maximum=2)
        B=1
        Alpha=&MaskFloatBySpawnGroup(Masked Float=1, Passthrough Float=1, Spawn Group Mask=1)
       )
       "[1] Amplitude Min/Max"=<0.5 1>
      )
     )
    )
    ScaleA=true
    ScaleRGB=true
   ) {module=ScaleColor001}
   &ScaleSpriteSizeBySpeed(
    Max Scale Factor=<0.2 0.6>
    Min Scale Factor=<0.2 0.2>
    Sample Scale Factor By Curve=true
    Scale Factor Curve=DataInterface() {propertyValues {Curve {keys [
        {interpMode=RCIM_Cubic, tangentMode=RCTM_User, time=0, value=0}
        {arriveTangent=2.9664, interpMode=RCIM_Cubic, leaveTangent=2.9664, tangentMode=RCTM_User, time=1, value=1}
       ]}}}
    Velocity Threshold=500
   )
   &ScaleSpriteSize(Uniform Curve Sprite Scale=DataInterface() {propertyValues {Curve {keys [
        {arriveTangent=2.9868, interpMode=RCIM_Cubic, leaveTangent=2.9868, tangentMode=RCTM_User, time=0, value=1.65}
        <at 0.2 2.25>
        <at 0.9 2.15>
        {arriveTangent=-0.0405, interpMode=RCIM_Cubic, leaveTangent=-0.0405, tangentMode=RCTM_User, time=1, value=1.4}
       ]}}})
  ]
  rendererClasses [&NiagaraSpriteRendererProperties]
  renderers [{rendererClass=&NiagaraSpriteRendererProperties}]
 }
}
summary {
 systemName=NS_Thermite_Explosion
 emitters [
  {emitterName=LW_Sparks1, rendererClasses []}
  {emitterName=LW_Debris, rendererClasses []}
  {emitterName=LW_Flash, rendererClasses []}
  {emitterName=LW_Fire, rendererClasses []}
  {emitterName=LW_Streaks, rendererClasses []}
  {emitterName=TrailEmbers_Hi_Ep001, simTarget=GPUComputeSim, rendererClasses [&NiagaraSpriteRendererProperties]}
  {emitterName=Dark_Chunks, rendererClasses []}
  {emitterName=Lit001, rendererClasses [&NiagaraLightRendererProperties, &NiagaraDecalRendererProperties]}
  {emitterName=SourceParticles, simTarget=GPUComputeSim, rendererClasses [&NiagaraSpriteRendererProperties]}
  {emitterName=Grid2D_Gas_SmokeFire_Emitter, simTarget=GPUComputeSim, rendererClasses [&NiagaraSpriteRendererProperties]}
  {emitterName=LW_Explosion_Main, rendererClasses []}
  {emitterName=LW_GlowSphereMesh2, rendererClasses []}
 ]
 userVariables [
  User.Global Scale {type=NiagaraFloat, defaultValue=1, description=""}
  User.IsTrain {type=NiagaraBool}
 ]
}
