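# Niagara system for the heating-up effect of thermite, with emitters named for fire smoke, flash, sparks, streaks and light.
# Only the Lit emitter shows modules: a light renderer, a color curve on emitter age that quickly ramps to intense red-orange, and a sine-flickered radius.
# Lit also derives a decal visibility flag as the inverse of User.IsTrain, though no decal renderer appears.
# The four LW_ emitters list no modules or renderers here.
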NiagaraSystem /ThermiteGameplay/Assets/FX/NS_Thermite_HeatingUp
dependencies {
 usedDataInterfaces [&NiagaraDataInterfaceColorCurve, &NiagaraDataInterfaceCurve]
 usedDynamicInputs [
  &Multiply_VectorByFloat
  &FloatFromCurve
  &Divide_Float
  &Lerp_Float
 ]
 usedModules [
  &self:SystemSpawnScript.SystemScriptSource.SystemScriptGraph.NiagaraNodeAssignment_0.SetVariables_#1
  &SystemState
  &self:SystemSpawnScript.SystemScriptSource.SystemScriptGraph.NiagaraNodeAssignment_1.SetVariables_#2
 ]
}
emitters {
 LW_FireSmoke {rendererClasses [], renderers []}
 LW_Flash {rendererClasses [], renderers []}
 LW_Sparks1 {rendererClasses [], renderers []}
 LW_Streaks {rendererClasses [], renderers []}
 Lit {
  emitterSpawnScript [
   /SetVariables(Emitter.Position:NiagaraPosition=<0 0 200>, Emitter.Position2:NiagaraPosition=<0 0 50>) {isSetParameters=true}
   /SetVariables(Emitter.DecalVisible:NiagaraBool=&Boolean_NotOperation(Bool=$User.IsTrain)) {isSetParameters=true}
  ]
  emitterUpdateScript [
   &EmitterState(
    Inactive Response=::NewEnumerator1
    Life Cycle Mode=::NewEnumerator1
    Loop Behavior=::NewEnumerator1
    Loop Delay=0.9
    Loop Duration Mode=::NewEnumerator1
    UseLoopDelay=false
   )
   /SetVariables(
    Emitter.LightColor:LinearColor=&ColorFromCurve(
     ColorCurve=DataInterface() {propertyValues {
       AlphaCurve {keys [<at 0.0069 0>,<at 0.0637 1 RCIM_Cubic>]}
       BlueCurve {keys [<at 0.0115 0>,<at 0.0599 0.0588 RCIM_Cubic>]}
       GreenCurve {keys [<at 0.0115 0>,<at 0.0599 7.946 RCIM_Cubic>]}
       RedCurve {keys [<at 0.0115 0>,<at 0.0599 100 RCIM_Cubic>]}
      }}
     CurveIndex=$Emitter.Age
    )
    Emitter.LightRadius:NiagaraFloat=&Multiply_Float(
     A=70
     B=&Sine(
      Normalized Angle=$Emitter.Age
      Period=&RandomRangeFloat(Minimum=0.2, Maximum=0.4, Evaluation Type=::NewEnumerator1)
      Scale=0.05
      Bias=1
     )
    )
   ) {isSetParameters=true}
  ]
  rendererClasses [&NiagaraLightRendererProperties]
  renderers [{rendererClass=&NiagaraLightRendererProperties, rendererIndex=0}]
 }
}
summary {
 systemName=NS_Thermite_HeatingUp
 emitters [
  {emitterName=LW_Sparks1, rendererClasses []}
  {emitterName=LW_Flash, rendererClasses []}
  {emitterName=LW_Streaks, rendererClasses []}
  {emitterName=Lit, rendererClasses [&NiagaraLightRendererProperties]}
  {emitterName=LW_FireSmoke, rendererClasses []}
 ]
 userVariables [
  User.Duration {type=NiagaraFloat, defaultValue=15, description=""}
  User.IsTrain {type=NiagaraBool}
  User.Radius {type=NiagaraFloat, defaultValue=300, description=""}
  User.Skeletal Mesh3 {
   type=NiagaraDataInterfaceSkeletalMesh
   defaultValue {
    struct=NiagaraExt_VariableValue_DataInterface
    value {dataInterface=&self:NiagaraDataInterfaceSkeletalMesh_0}
   }
  }
 ]
}
