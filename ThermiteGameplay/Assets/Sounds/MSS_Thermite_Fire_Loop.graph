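# Stereo MetaSound source for the thermite gadget's looping fire sound, with an initial stage and a burning stage.
# On play, it fires a random one-shot and loop A; the OnBurning trigger fires another one-shot plus loops B and C.
# Trigger routes feed short interpolated mixer gains that crossfade loop A into B, while C keeps fixed gain.
# Attenuation distance is mapped to fade the three loops, but not the one-shots.
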MetaSoundSource /ThermiteGameplay/Assets/Sounds/MSS_Thermite_Fire_Loop
inputs [
 OnBurning {type=Trigger}
 UE.Attenuation.Distance {type=Float}
 UE.Source.OnPlay {type=Trigger}
]
interfaces [UE.Attenuation:1.0, UE.OutputFormat.Stereo:1.0, UE.Source:1.1]
outputs [
 "UE.OutputFormat.Stereo.Audio:0" {type=Audio}
 "UE.OutputFormat.Stereo.Audio:1" {type=Audio}
]
pages [Default {
  input_defaults {OnBurning=null, UE.Attenuation.Distance=0, UE.Source.OnPlay=null}
  nodes [
   n1 = &MSS_PlayRandomSound_Stereo(UE.Source.OnPlay=^OnBurning, Sounds=VariableAccessor.WaveAsset:Array(Variable=InitVariable.WaveAsset:Array().Variable).Value) {also {defaults {Volume=1.5}}}
   n2 = &MSS_PlayRandomSound_Stereo(UE.Source.OnPlay=^UE.Source.OnPlay, Sounds=VariableAccessor.WaveAsset:Array(Variable=InitVariable.WaveAsset:Array().Variable).Value)
   n3 = MapRange.MapRange:Float(In=^UE.Attenuation.Distance) {also {defaults {In Range B=6000, Out Range A=1, Out Range B=0}}}
   WavePlayer_1 = UE.Wave Player:Stereo(Play=^UE.Source.OnPlay) {also {defaults {Loop="<MetaSoundBuilderResult.SUCCEEDED: 0>", Wave Asset=&/Thermite_Default_Loop_A}}}
   n4 = VariableAccessor.Float(Variable=InitVariable.Float().Variable)
   WavePlayer_2 = UE.Wave Player:Stereo(Play=^OnBurning) {also {defaults {Loop="<MetaSoundBuilderResult.SUCCEEDED: 0>", Wave Asset=&/Thermite_Default_Loop_C}}}
   WavePlayer_0 = UE.Wave Player:Stereo(Play=^OnBurning) {also {defaults {Loop="<MetaSoundBuilderResult.SUCCEEDED: 0>", Wave Asset=&/Thermite_Default_Loop_B}}}
   n5 = VariableAccessor.Float(Variable=InitVariable.Float().Variable)
   n6 = InitVariable.Float()
   TriggerRouteTriggerRouteFloat2_0 = "TriggerRoute.Trigger Route (Float, 2):Float"(Set 1=^OnBurning, Set 0=^UE.Source.OnPlay) {also {defaults {Value 0=0, Value 1=0.7}}}
   n7 = UE.InterpTo:Audio(Target=@TriggerRouteTriggerRouteFloat2_0.Value) {also {defaults {Interp Time=0.2}}}
   n8 = InitVariable.Float()
   TriggerRouteTriggerRouteFloat2_1 = "TriggerRoute.Trigger Route (Float, 2):Float"(Set 1=^OnBurning, Set 0=^UE.Source.OnPlay) {also {defaults {Value 0=1}}}
   n9 = UE.InterpTo:Audio(Target=@TriggerRouteTriggerRouteFloat2_1.Value) {also {defaults {Interp Time=0.2}}}
   AudioMixerAudioMixerStereo5_0 = "AudioMixer.Audio Mixer (Stereo, 5)"(
    In 2 L=@n1."UE.OutputFormat.Stereo.Audio:0"
    In 2 R=@n1."UE.OutputFormat.Stereo.Audio:1"
    In 0 L=@n2."UE.OutputFormat.Stereo.Audio:0"
    In 0 R=@n2."UE.OutputFormat.Stereo.Audio:1"
    In 1 L=UE.Multiply:Audio by Float(
     PrimaryOperand=UE.Multiply:Audio by Float(AdditionalOperands=@n3.Out Value, PrimaryOperand=@WavePlayer_1.Out Left).Out
     AdditionalOperands=@n4.Value
    ).Out
    In 4 R=UE.Multiply:Audio by Float(AdditionalOperands=@n3.Out Value, PrimaryOperand=@WavePlayer_2.Out Right).Out
    In 3 R=UE.Multiply:Audio by Float(
     PrimaryOperand=UE.Multiply:Audio by Float(AdditionalOperands=@n3.Out Value, PrimaryOperand=@WavePlayer_0.Out Right).Out
     AdditionalOperands=@n5.Value
    ).Out
    In 4 L=UE.Multiply:Audio by Float(AdditionalOperands=@n3.Out Value, PrimaryOperand=@WavePlayer_2.Out Left).Out
    In 1 R=UE.Multiply:Audio by Float(
     PrimaryOperand=UE.Multiply:Audio by Float(AdditionalOperands=@n3.Out Value, PrimaryOperand=@WavePlayer_1.Out Right).Out
     AdditionalOperands=@n4.Value
    ).Out
    In 3 L=UE.Multiply:Audio by Float(
     PrimaryOperand=UE.Multiply:Audio by Float(AdditionalOperands=@n3.Out Value, PrimaryOperand=@WavePlayer_0.Out Left).Out
     AdditionalOperands=@n5.Value
    ).Out
    Gain 3=VariableAccessor.Float(Variable=VariableMutator.Float(Variable=@n6.Variable, Value=@n7.Value).Variable).Value
    Gain 1=VariableAccessor.Float(Variable=VariableMutator.Float(Variable=@n8.Variable, Value=@n9.Value).Variable).Value
   ) {also {defaults {Gain 4=1.5}}}
  ]
  links [
   {dst="@UE.OutputFormat.Stereo.Audio:0", src=AudioMixerAudioMixerStereo5_0.Out L}
   {dst="@UE.OutputFormat.Stereo.Audio:1", src=AudioMixerAudioMixerStereo5_0.Out R}
  ]
 }]
properties {
 assetImportData=&self:AssetImportData
 attenuationSettings=&/ATT_Thermite
 bLooping=true
 bOverrideConcurrency=true
 duration=10000
 maxDistance=8000
 numChannels=2
 outputFormat=Stereo
 sampleRate=48000
 soundClassObject=&Fort_Gadgets
 volume=0.7
 concurrencyOverrides {maxCount=3}
}
