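# Sound cue for the Thermite cook sound, playing one of two cook wave variants at random.
# A Random node with equal weights picks the wave and feeds a Modulator that varies pitch slightly while keeping volume fixed.
# It borrows the weapon reload attenuation and the loot collect sound class, with a boosted overall volume.
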SoundCue /ThermiteGameplay/Assets/Sounds/SC_Thermite_Cook_Cue
root=Modulator_0
nodes [
 {id=Modulator_0, children [Random_0], properties {pitchMax=1.1, pitchMin=0.9, volumeMax=1, volumeMin=1}}
 {id=Random_0, children [WavePlayer_1,WavePlayer_2], properties {bShouldExcludeFromBranchCulling=true, weights [1,1]}}
 {id=WavePlayer_1, properties {soundWaveAssetPtr=&/Thermite_Cook_1}}
 {id=WavePlayer_2, properties {soundWaveAssetPtr=&/Thermite_Cook_2}}
]
properties {
 attenuationSettings=&Fort_Weapon_Reload_Att
 maxDistance=2250
 soundClassObject=&Fort_Loot_Collect
 volumeMultiplier=1.4
}
