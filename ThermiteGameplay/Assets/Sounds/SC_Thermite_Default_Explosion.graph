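# Sound cue for the default Thermite explosion, classed as a gadget sound, with explosion attenuation and an overridden concurrency limit.
# A mixer layers a detonation wave, slightly delayed and at half volume, with one of three equally weighted explosion waves at full volume.
# Each layer passes through its own modulator, which randomly raises pitch for variation while volume stays constant.
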SoundCue /ThermiteGameplay/Assets/Sounds/SC_Thermite_Default_Explosion
root=Mixer_0
nodes [
 {id=Delay_1, children [Modulator_0], properties {delayMax=0.08, delayMin=0.08}}
 {id=Mixer_0, children [Delay_1,Modulator_1], properties {inputVolume [0.5,1]}}
 {id=Modulator_0, children [WavePlayer_0], properties {pitchMax=1.2, pitchMin=1, volumeMax=1, volumeMin=1}}
 {id=Modulator_1, children [Random_0], properties {pitchMax=1.2, pitchMin=1, volumeMax=1, volumeMin=1}}
 {id=Random_0, children [WavePlayer_1,WavePlayer_2,WavePlayer_3], properties {bShouldExcludeFromBranchCulling=true, weights [1,1,1]}}
 {id=WavePlayer_0, properties {soundWaveAssetPtr=&/Thermite_Default_Detonation}}
 {id=WavePlayer_1, properties {soundWaveAssetPtr=&/Thermite_Explosion_1}}
 {id=WavePlayer_2, properties {soundWaveAssetPtr=&/Thermite_Explosion_2}}
 {id=WavePlayer_3, properties {soundWaveAssetPtr=&/Thermite_Explosion_3}}
]
properties {
 attenuationSettings=&/ATT_Thermite_Explo
 bOverrideConcurrency=true
 maxDistance=16000
 soundClassObject=&Fort_Gadgets
 volumeMultiplier=1
 concurrencyOverrides {maxCount=3}
}
