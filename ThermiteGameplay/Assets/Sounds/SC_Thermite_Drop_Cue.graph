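# Sound cue for dropping the Thermite item, part of the ThermiteGameplay content.
# A Random node at the root picks one of two drop wave variations with equal weight and is excluded from branch culling.
# The cue plays at reduced volume through the shared weapon-drop attenuation and the loot-collect sound class.
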SoundCue /ThermiteGameplay/Assets/Sounds/SC_Thermite_Drop_Cue
root=Random_0
nodes [
 {id=Random_0, children [WavePlayer_0,WavePlayer_3], properties {bShouldExcludeFromBranchCulling=true, weights [1,1]}}
 {id=WavePlayer_0, properties {soundWaveAssetPtr=&/Thermite_Drop_1}}
 {id=WavePlayer_3, properties {soundWaveAssetPtr=&/Thermite_Drop_2}}
]
properties {
 attenuationSettings=&Fort_Weapon_Drop_Att
 maxDistance=3000
 soundClassObject=&Fort_Loot_Collect
 volumeMultiplier=0.4
}
