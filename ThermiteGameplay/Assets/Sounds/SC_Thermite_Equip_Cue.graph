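# Sound cue for equipping Thermite, playing one of two equip wave variants.
# A Random node picks between the two waves with equal weight and feeds a Modulator at the root for playback variation.
# It uses the weapon reload attenuation and loot collect sound class, with overridden concurrency that caps simultaneous instances.
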SoundCue /ThermiteGameplay/Assets/Sounds/SC_Thermite_Equip_Cue
root=Modulator_0
nodes [
 {id=Modulator_0, children [Random_0]}
 {id=Random_0, children [WavePlayer_1,WavePlayer_0], properties {weights [1,1]}}
 {id=WavePlayer_0, properties {soundWaveAssetPtr=&/Thermite_Equip_1}}
 {id=WavePlayer_1, properties {soundWaveAssetPtr=&/Thermite_Equip_2}}
]
properties {
 attenuationSettings=&Fort_Weapon_Reload_Att
 bOverrideConcurrency=true
 maxDistance=2250
 soundClassObject=&Fort_Loot_Collect
 concurrencyOverrides {maxCount=3}
}
