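# Sound cue for picking up the Thermite item, playing one of two pickup recordings.
# An equal-weight Random node picks the wave, and a Modulator at default settings feeds the output.
# It uses the shared weapon pickup attenuation and the gadgets/items sound class at slightly reduced volume, with a concurrency override capping simultaneous plays.
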SoundCue /ThermiteGameplay/Assets/Sounds/SC_Thermite_Pickup_Cue
root=Modulator_0
nodes [
 {id=Modulator_0, children [Random_0]}
 {id=Random_0, children [WavePlayer_1,WavePlayer_3], properties {weights [1,1]}}
 {id=WavePlayer_1, properties {soundWaveAssetPtr=&/Thermite_PickUp_1}}
 {id=WavePlayer_3, properties {soundWaveAssetPtr=&/Thermite_PickUp_2}}
]
properties {
 attenuationSettings=&Fort_Weapon_Pickup_Att
 bOverrideConcurrency=true
 maxDistance=3500
 soundClassObject=&Fort_Gadgets_Items
 volumeMultiplier=0.8
 concurrencyOverrides {maxCount=4}
}
