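# Sound cue for a thrown thermite projectile, mixing a one-shot launch sound with a looping in-air sound.
# The boosted launch branch layers a thermite throw with a random frag-grenade throw, which the last quality-level slot omits.
# The looping wave fades in through an Enveloper volume curve.
# Each branch has its own attenuation node, with the launch overriding its asset for long-range falloff, HRTF and a distant reverb send.
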SoundCue /ThermiteGameplay/Assets/Sounds/Thermite_InAirLoop_Cue
root=Mixer_1
nodes [
 {id=Attenuation_1, children [Enveloper_0], properties {attenuationSettings=/Game/Sounds/Attenuation/Fort_Weapon_Projectile_InAir_Loop_Att}}
 {
  id=Attenuation_2
  children [Mixer_2]
  properties {
   attenuationSettings=/Game/Sounds/Attenuation/Fort_Weapon_Projectile_InAir_Launch_Throw_Att
   bOverrideAttenuation=true
   attenuationOverrides {
    bAttenuateWithLPF=true
    bEnableSubmixSends=true
    falloffDistance=30000
    hPFFrequencyAtMax=100
    lPFFrequencyAtMax=1500
    lPFRadiusMax=1250
    lPFRadiusMin=0
    spatializationAlgorithm=SPATIALIZATION_HRTF
    stereoSpread=0
    attenuationShapeExtents {x=0}
    submixSendSettings [{
      bEnableHPFCutoff=false
      bEnableLPFCutoff=false
      disableManualSendClamp=false
      hPFCutoff=20
      lPFCutoff=20000
      maxSendDistance=7500
      maxSendLevel=0.25
      minSendDistance=0
      minSendLevel=0
      sendLevel=0.2
      sendLevelControlMethod=Linear
      soundSubmix=&Fortnite_Reverb_Distant
      customSendLevelCurve {externalCurve=/Game/Sounds/SubMixes/Reverb/Fortnite_Reverb_Distant_Send}
     }]
   }
  }
 }
 {
  id=Enveloper_0
  children [Modulator_0]
  properties {pitchCurve {editorCurveData {keys [{time=0}]}}, volumeCurve {editorCurveData {keys [<at 0.0001 0.0145>,<at 0.1225 0.0867>,<at 1 1>]}}}
 }
 {id=Mixer_1, children [Attenuation_2, Attenuation_1], properties {inputVolume [1.5,1]}}
 {id=Mixer_2, children [Modulator_3, QualityLevel_0], properties {inputVolume [1,1]}}
 {id=Modulator_0, children [WavePlayer_0]}
 {id=Modulator_2, children [Random_1]}
 {id=Modulator_3, children [WavePlayer_2]}
 {id=QualityLevel_0, children [Modulator_2,Modulator_2,Modulator_2,null]}
 {id=Random_1, children [WavePlayer_4,WavePlayer_5,WavePlayer_6], properties {bShouldExcludeFromBranchCulling=true, weights [1,1,1]}}
 {id=WavePlayer_0, properties {bLooping=true, soundWaveAssetPtr=&/Thermite_InAirLoop}}
 {id=WavePlayer_2, properties {soundWaveAssetPtr=&/Thermite_Throw_1}}
 {id=WavePlayer_4, properties {soundWaveAssetPtr=/Game/Sounds/Fort_Weapons/Grenades/Frag/Grenade_Frag_Throw_01}}
 {id=WavePlayer_5, properties {soundWaveAssetPtr=/Game/Sounds/Fort_Weapons/Grenades/Frag/Grenade_Frag_Throw_02}}
 {id=WavePlayer_6, properties {soundWaveAssetPtr=/Game/Sounds/Fort_Weapons/Grenades/Frag/Grenade_Frag_Throw_03}}
]
properties {
 attenuationSettings=&Fort_Weapon_Projectile_InAir_Loop_Att
 maxDistance=30000
 volumeMultiplier=1
 attenuationOverrides {
  bAttenuateWithLPF=true
  bEnableSubmixSends=true
  falloffDistance=8000
  hPFFrequencyAtMax=100
  lPFFrequencyAtMax=1500
  lPFRadiusMax=1250
  lPFRadiusMin=0
  reverbDistanceMax=1024
  reverbDistanceMin=0
  reverbWetLevelMax=0
  reverbWetLevelMin=1
  spatializationAlgorithm=SPATIALIZATION_HRTF
  stereoSpread=0
  attenuationShapeExtents {x=2000}
  submixSendSettings [{
    bEnableHPFCutoff=false
    bEnableLPFCutoff=false
    disableManualSendClamp=false
    hPFCutoff=20
    lPFCutoff=20000
    maxSendDistance=7500
    maxSendLevel=0.25
    minSendDistance=0
    minSendLevel=0
    sendLevel=0.2
    sendLevelControlMethod=Linear
    soundSubmix=&Fortnite_Reverb_Distant
    customSendLevelCurve {externalCurve=&Fortnite_Reverb_Distant_Send}
   }]
 }
}
